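# Control Rig for the F_MED_Sailor_Squad_Rose character body that procedurally poses the headset wire.
# It aims dyn_driver_headset_wire at the coat wire's parent bone, carrying its children along.
# Head rotation relative to spine_05 is converted to Euler; the larger of two clamped X and Y remaps weights an additive local bend on dyn_headset_wire_01.
# Debug axis draws on the wire bone remain enabled before and after the adjustment.
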ControlRigBlueprint /BRCosmetics/Characters/Player/Female/Medium/Bodies/F_MED_Sailor_Squad_Rose/Meshes/F_MED_Sailor_Squad_Rose_CtrlRig
graphs [
 RigVMFunctionLibrary {links [], nodes []}
 RigVMModel {nodes [
   RigUnit_BeginExecution = &RigUnit_BeginExecution()
   /Script/ControlRig.RigUnit_DebugTransformMutable(
    Color=(R=1,G=0,B=0,A=1)
    DebugDrawSettings=(DepthPriority=SDPG_Foreground,Lifetime=-1)
    Mode=Axes
    Scale=10
    Space=None
    Thickness=0.5
    WorldOffset=(Rotation=(X=0,Y=0,Z=0,W=1),Translation=(X=0,Y=0,Z=0),Scale3D=(X=1,Y=1,Z=1))
    bEnabled=True
    ExecutePin=@RigUnit_BeginExecution.ExecutePin
    Transform=&RigUnit_GetBoneTransform(Bone=dyn_headset_wire_01).Transform
   ) {is=struct RigVMUnitNode}
   RigUnit_GetBoneTransform = &RigUnit_GetBoneTransform(Bone=dyn_parent_wire_coat_01)
   > /Script/ControlRig.RigUnit_AimBone(
    Bone=dyn_driver_headset_wire
    DebugSettings=(Scale=10,WorldOffset=(Rotation=(X=0,Y=0,Z=0,W=1),Translation=(X=0,Y=0,Z=0),Scale3D=(X=1,Y=1,Z=1)))
    Primary=(Weight=1,Axis=(X=1,Y=0,Z=0),Target=(X=1,Y=0,Z=0),Kind=Location)
    Secondary=(Weight=0,Axis=(X=0,Y=0,Z=0),Target=(X=1,Y=0,Z=0),Kind=Location)
    Weight=1
    bPropagateToChildren=True
    Primary.Target=@RigUnit_GetBoneTransform.Transform.Translation
   ) {is=struct RigVMUnitNode}
   > &RigUnit_DebugTransformMutable(
    Color=(R=1,G=0,B=0,A=1)
    DebugDrawSettings=(DepthPriority=SDPG_Foreground,Lifetime=-1)
    Mode=Axes
    Scale=10
    Space=None
    Thickness=1
    WorldOffset=(Rotation=(X=0,Y=0,Z=0,W=1),Translation=(X=0,Y=0,Z=0),Scale3D=(X=1,Y=1,Z=1))
    bEnabled=false
    Transform=&RigUnit_GetBoneTransform(Bone=dyn_driver_headset_wire).Transform
   )
   n1 = &RigVMFunction_MathQuaternionToEuler(RotationOrder=XYZ, Value=&RigUnit_GetRelativeBoneTransform(Bone=head, Space=spine_05).Transform.Rotation)
   > /Script/ControlRig.RigUnit_ModifyBoneTransforms(
    BoneToModify=((Bone="dyn_headset_wire_01",Transform=(Rotation=(X=0,Y=0,Z=0,W=1),Translation=(X=0,Y=0,Z=0),Scale3D=(X=1,Y=1,Z=1))))
    Mode=AdditiveLocal
    WeightMaximum=1
    Weight=/Script/RigVM.RigVMFunction_MathFloatMax(
     A=&RigVMFunction_MathFloatRemap(SourceMaximum=50, SourceMinimum=0, TargetMinimum=0, bClamp=true, Value=@n1.Result.Y).Result
     B=&RigVMFunction_MathFloatRemap(SourceMaximum=-45, SourceMinimum=0, TargetMinimum=0, bClamp=true, Value=@n1.Result.X).Result
    ) {is=struct RigVMUnitNode, template="Maximum::Execute(in A,in B,out Result)"}.Result
    BoneToModify.0.Transform.Rotation=&RigVMFunction_MathQuaternionFromEuler(Euler=(X=0,Y=40,Z=0)).Result
   ) {is=struct RigVMUnitNode}
   > &RigUnit_DebugTransformMutable(
    Color=(R=1,G=0,B=0,A=1)
    DebugDrawSettings=(DepthPriority=SDPG_Foreground,Lifetime=-1)
    Mode=Axes
    Scale=10
    Space=None
    Thickness=1
    WorldOffset=(Rotation=(X=0,Y=0,Z=0,W=1),Translation=(X=0,Y=0,Z=0),Scale3D=(X=1,Y=1,Z=1))
    bEnabled=True
    Transform=&RigUnit_GetBoneTransform(Bone=dyn_headset_wire_01).Transform
   )
  ]}
]
hierarchy [
 root {loc=<0 0 0>, rot=<0 0 0 1>}
 pelvis {parent=root, loc=<0 -0.5301 98.0571>, rot=<0.0127 0.707 -0.0127 -0.707>}
 spine_01 {parent=pelvis, loc=<6.1208 -0.9328 0>, rot=<0 0 0.034 -0.9994>}
 spine_02 {parent=spine_01, loc=<6.5996 -0.2095 0>, rot=<0 0 0.0238 -0.9997>}
 spine_03 {parent=spine_02, loc=<6.4914 -0.1237 0>, rot=<0 0 0.0149 -0.9999>}
 spine_04 {parent=spine_03, loc=<6.9598 0.0794 0>, rot=<0 0 -0.0132 -0.9999>}
 spine_05 {parent=spine_04, loc=<6.9987 0 0>, rot=<0 0 -0.0015 -1>}
 clavicle_l {parent=spine_05, loc=<5.4831 -0.1053 -3.2946>, rot=<0.0465 -0.7289 -0.009 -0.6829>}
 upperarm_l {parent=clavicle_l, loc=<9.8658 -1.2136 -0.4105>, rot=<0.0164 -0.3245 -0.0269 -0.9454>}
 lowerarm_l {parent=upperarm_l, loc=<26.465 -0.8345 0.259>, rot=<0.0036 0.0161 -0.2298 -0.9731>}
 hand_l {parent=lowerarm_l, loc=<22.7355 -0.3203 -0.6105>, rot=<-0.6784 -0.0223 0.0735 -0.7307>}
 index_metacarpal_l {parent=hand_l, loc=<1.6535 0.0183 -2.96>, rot=<0.0158 -0.0124 -0.0173 -0.9996>}
 index_01_l {parent=index_metacarpal_l, loc=<8.3008 0.0357 -0.0524>, rot=<-0.0044 0.0485 0.0924 -0.9945>}
 index_02_l {parent=index_01_l, loc=<3.3587 -0.3003 0.0452>, rot=<-0.002 0.0027 0.3225 -0.9466>}
 index_03_l {parent=index_02_l, loc=<2.5578 -0.123 -0.0252>, rot=<-0.002 0.0067 0.2558 -0.9667>}
 middle_metacarpal_l {parent=hand_l, loc=<2.1856 -0.1725 -0.5984>, rot=<0.006 0.0168 -0.0329 -0.9993>}
 middle_01_l {parent=middle_metacarpal_l, loc=<7.3275 -0.2447 0.0129>, rot=<-0.0208 0.0042 0.0858 -0.9961>}
 middle_02_l {parent=middle_01_l, loc=<4.0123 0.1401 0.1175>, rot=<-0.0218 0.0258 0.3512 -0.9357>}
 middle_03_l {parent=middle_02_l, loc=<2.6077 0.0534 0.1373>, rot=<0.0082 0.0188 0.1708 -0.9851>}
 pinky_metacarpal_l {parent=hand_l, loc=<1.7616 -0.9753 3.3814>, rot=<-0.0397 0.0814 -0.069 -0.9935>}
 pinky_01_l {parent=pinky_metacarpal_l, loc=<7.1709 -0.4831 -0.1094>, rot=<-0.0654 -0.0035 0.2242 -0.9723>}
 pinky_02_l {parent=pinky_01_l, loc=<2.8274 0.0893 -0.1043>, rot=<0.0037 0.0011 0.2854 -0.9584>}
 pinky_03_l {parent=pinky_02_l, loc=<2.2046 0.0307 -0.0078>, rot=<-0.0002 0.0098 0.1445 -0.9895>}
 ring_metacarpal_l {parent=hand_l, loc=<2.2074 -0.3625 1.3638>, rot=<-0.0284 0.0722 -0.0458 -0.9959>}
 ring_01_l {parent=ring_metacarpal_l, loc=<7.0295 -0.6197 -0.1446>, rot=<-0.0489 -0.0252 0.1892 -0.9804>}
 ring_02_l {parent=ring_01_l, loc=<3.6174 0.4647 -0.1027>, rot=<0 0 0.2806 -0.9598>}
 ring_03_l {parent=ring_02_l, loc=<2.4152 0.114 0.0199>, rot=<0 0 0.1189 -0.9929>}
 thumb_01_l {parent=hand_l, loc=<1.3118 -0.3191 -2.459>, rot=<0.6546 -0.4042 0.0664 -0.6354>}
 thumb_02_l {parent=thumb_01_l, loc=<5.9336 -0.0448 0.0401>, rot=<-0.0037 0.0213 0.298 -0.9543>}
 thumb_03_l {parent=thumb_02_l, loc=<2.2223 0.3682 -0.0907>, rot=<0.0118 -0.013 0.0066 -0.9998>}
 lowerarm_twist_01_l {parent=lowerarm_l, loc=<18.316 0 0>, rot=<0 0 0 1>}
 lowerarm_twist_02_l {parent=lowerarm_l, loc=<10.6707 0 0>, rot=<0 0 0 1>}
 upperarm_twist_01_l {parent=upperarm_l, loc=<7.2135 -0.1134 0>, rot=<0 0 0 -1>}
 upperarm_twist_02_l {parent=upperarm_l, loc=<19.7998 -0.1134 0>, rot=<0 0 0 -1>}
 clavicle_r {parent=spine_05, loc=<5.4831 -0.1053 3.2946>, rot=<0.7289 0.0465 -0.6829 0.009>}
 upperarm_r {parent=clavicle_r, loc=<-9.8658 1.2136 0.4104>, rot=<0.0164 -0.3245 -0.0269 -0.9454>}
 lowerarm_r {parent=upperarm_r, loc=<-26.4651 0.8345 -0.259>, rot=<0.0036 0.0161 -0.2298 -0.9731>}
 hand_r {parent=lowerarm_r, loc=<-22.7356 0.3203 0.6106>, rot=<-0.6784 -0.0223 0.0735 -0.7307>}
 index_metacarpal_r {parent=hand_r, loc=<-1.6534 -0.0181 2.9601>, rot=<0.0158 -0.0124 -0.0173 -0.9996>}
 index_01_r {parent=index_metacarpal_r, loc=<-8.3008 -0.0357 0.0524>, rot=<-0.0044 0.0485 0.0924 -0.9945>}
 index_02_r {parent=index_01_r, loc=<-3.3589 0.3003 -0.0453>, rot=<-0.002 0.0027 0.3225 -0.9466>}
 index_03_r {parent=index_02_r, loc=<-2.5578 0.123 0.0252>, rot=<-0.002 0.0067 0.2558 -0.9667>}
 middle_metacarpal_r {parent=hand_r, loc=<-2.1857 0.1726 0.5984>, rot=<0.006 0.0168 -0.0329 -0.9993>}
 middle_01_r {parent=middle_metacarpal_r, loc=<-7.3273 0.2447 -0.0129>, rot=<-0.0208 0.0042 0.0858 -0.9961>}
 middle_02_r {parent=middle_01_r, loc=<-4.0124 -0.14 -0.1175>, rot=<-0.0218 0.0258 0.3512 -0.9357>}
 middle_03_r {parent=middle_02_r, loc=<-2.6077 -0.0533 -0.1373>, rot=<0.0082 0.0188 0.1708 -0.9851>}
 pinky_metacarpal_r {parent=hand_r, loc=<-1.7617 0.9753 -3.3814>, rot=<-0.0397 0.0814 -0.069 -0.9935>}
 pinky_01_r {parent=pinky_metacarpal_r, loc=<-7.1707 0.4832 0.1094>, rot=<-0.0654 -0.0035 0.2242 -0.9723>}
 pinky_02_r {parent=pinky_01_r, loc=<-2.8274 -0.0892 0.1043>, rot=<0.0037 0.0011 0.2854 -0.9584>}
 pinky_03_r {parent=pinky_02_r, loc=<-2.2045 -0.031 0.0078>, rot=<-0.0002 0.0098 0.1445 -0.9895>}
 ring_metacarpal_r {parent=hand_r, loc=<-2.2073 0.3627 -1.3638>, rot=<-0.0284 0.0722 -0.0458 -0.9959>}
 ring_01_r {parent=ring_metacarpal_r, loc=<-7.0297 0.6197 0.1446>, rot=<-0.0489 -0.0252 0.1892 -0.9804>}
 ring_02_r {parent=ring_01_r, loc=<-3.6171 -0.4648 0.1027>, rot=<0 0 0.2806 -0.9598>}
 ring_03_r {parent=ring_02_r, loc=<-2.4153 -0.1141 -0.0199>, rot=<0 0 0.1189 -0.9929>}
 thumb_01_r {parent=hand_r, loc=<-1.3119 0.3193 2.459>, rot=<0.6546 -0.4042 0.0664 -0.6354>}
 thumb_02_r {parent=thumb_01_r, loc=<-5.9335 0.0447 -0.0401>, rot=<-0.0037 0.0213 0.298 -0.9543>}
 thumb_03_r {parent=thumb_02_r, loc=<-2.2223 -0.3682 0.0907>, rot=<0.0118 -0.013 0.0066 -0.9998>}
 lowerarm_twist_01_r {parent=lowerarm_r, loc=<-18.3161 0 0>, rot=<0 0 0 1>}
 lowerarm_twist_02_r {parent=lowerarm_r, loc=<-10.6707 0 0>, rot=<0 0 0 1>}
 upperarm_twist_01_r {parent=upperarm_r, loc=<-7.2135 0.1134 0>, rot=<0 0 0 1>}
 upperarm_twist_02_r {parent=upperarm_r, loc=<-19.7998 0.1134 0>, rot=<0 0 0 1>}
 neck_01 {parent=spine_05, loc=<11.3138 0.1529 0>, rot=<0 0 -0.2003 -0.9797>}
 neck_02 {parent=neck_01, loc=<4.5627 0 0>, rot=<0 0 0.0504 -0.9987>}
 head {parent=neck_02, loc=<5.3914 0 0>, rot=<0 0 0.115 -0.9934>}
 dyn_parent_headset_wire {parent=head, loc=<-3.7451 -0.7433 -9.6913>, rot=<0.1963 -0.1444 0.8344 0.4943>}
 dyn_driver_headset_wire {parent=dyn_parent_headset_wire, loc=<0 0 0>, rot=<0 0 0 1>}
 dyn_headset_wire_01 {parent=dyn_driver_headset_wire, loc=<0 0 0>, rot=<0 0 0 1>}
 dyn_frt_scarf {parent=spine_05, loc=<11.1251 -8.0434 0>, rot=<-0.0023 -0.0024 -0.7176 -0.6964>}
 dyn_parent_hood_frt_01_l {parent=spine_05, loc=<7.3905 4.8364 -8.545>, rot=<0.4382 0.1532 -0.1573 -0.8716>}
 dyn_driver_hood_frt_01_l {parent=dyn_parent_hood_frt_01_l, loc=<0 0 0>, rot=<0 0 0 -1>}
 dyn_hood_frt_01_l {parent=dyn_driver_hood_frt_01_l, loc=<0 0 0>, rot=<0 0 0 -1>}
 dyn_parent_hood_frt_01_r {parent=spine_05, loc=<7.3902 4.8364 8.545>, rot=<-0.4382 -0.1532 -0.1573 -0.8716>}
 dyn_driver_hood_frt_01_r {parent=dyn_parent_hood_frt_01_r, loc=<0 0 0>, rot=<0 0 0 1>}
 dyn_hood_frt_01_r {parent=dyn_driver_hood_frt_01_r, loc=<0 0 0>, rot=<0 0 0 1>}
 dyn_parent_hood_wire_01 {parent=spine_05, loc=<3.8283 8.2 -4.3194>, rot=<0.6796 0.0888 -0.116 -0.7189>}
 dyn_driver_hood_wire_01 {parent=dyn_parent_hood_wire_01, loc=<0 0 0>, rot=<0 0 0 -1>}
 dyn_hood_wire_01 {parent=dyn_driver_hood_wire_01, loc=<0 0 0>, rot=<0 0 0 -1>}
 dyn_parent_wire_coat_01 {parent=spine_05, loc=<10.6363 11.5712 -4.2814>, rot=<-0.0666 -0.9958 0.0093 0.0628>}
 dyn_driver_wire_coat_01 {parent=dyn_parent_wire_coat_01, loc=<0 0 0>, rot=<0 0 0 1>}
 dyn_wire_coat_01 {parent=dyn_driver_wire_coat_01, loc=<0 0 0>, rot=<0 0 0 1>}
 dyn_wire_coat_02 {parent=dyn_wire_coat_01, loc=<18.347 0 0>, rot=<-0.0341 -0.2351 0.0678 -0.969>}
 dyn_wire_coat_03 {parent=dyn_wire_coat_02, loc=<12.8008 0 0>, rot=<-0.0018 -0.028 0.0102 -0.9996>}
 dyn_parent_zipper {parent=spine_04, loc=<2.9095 10.3459 10.3746>, rot=<-0.0293 -0.2106 0.9658 0.1486>}
 dyn_zipper {parent=dyn_parent_zipper, loc=<0.2902 -0.5782 0.0001>, rot=<0 0 0 -1>}
 thigh_l {parent=pelvis, loc=<-1.0855 -0.2109 -8.7053>, rot=<0.0009 0.0355 0.0155 -0.9993>}
 calf_l {parent=thigh_l, loc=<-42.5284 -0.0019 -0.7297>, rot=<0.0006 -0.0161 -0.0391 -0.9991>}
 calf_twist_01_l {parent=calf_l, loc=<-35.8483 -1.9197 -0.5794>, rot=<0.0237 0.0095 -0.0035 -0.9997>}
 calf_twist_02_l {parent=calf_l, loc=<-19.1546 -1.7943 -0.2638>, rot=<0.0237 0.0095 -0.0035 -0.9997>}
 foot_l {parent=calf_l, loc=<-46.7401 -2.5014 -0.592>, rot=<0.0427 -0.0238 0.0353 -0.9982>}
 ball_l {parent=foot_l, loc=<-3.5658 10.8066 -0.1544>, rot=<0.0001 -0.0001 -0.7063 -0.7079>}
 dyn_parent_shoe_tag_l {parent=foot_l, loc=<6.4952 -5.7042 0.1041>, rot=<-0.0454 -0.0001 0.0639 -0.9969>}
 dyn_shoe_tag_01_l {parent=dyn_parent_shoe_tag_l, loc=<0 0 0>, rot=<0 0 0 -1>}
 thigh_twist_01_l {parent=thigh_l, loc=<-27.952 0 0>, rot=<0.0013 0 0 -1>}
 sub_thigh_l {parent=thigh_l, loc=<-2.972 0 0>, rot=<0 0 0 -1>}
 driver_frill_frt_l {parent=sub_thigh_l, loc=<6.7472 8.5623 -0.0995>, rot=<0 0 0.0087 -1>}
 dyn_frill_frt_l {parent=driver_frill_frt_l, loc=<0 0 0>, rot=<0 0 0 -1>}
 driver_frill_side_l {parent=sub_thigh_l, loc=<2.8149 -0.4685 -8.1075>, rot=<0 0 0 -1>}
 dyn_frill_side_l {parent=driver_frill_side_l, loc=<0 0 0>, rot=<0 0 0 -1>}
 driver_frill_bck_l {parent=sub_thigh_l, loc=<3.0963 -9.5518 -1.1757>, rot=<0 0 0 -1>}
 dyn_frill_bck_l {parent=driver_frill_bck_l, loc=<0 0 0>, rot=<0 0 0 -1>}
 dyn_parent_thigh_pack_l {parent=thigh_l, loc=<-29.5411 -0.8011 -6.1776>, rot=<-0.07 -0.7151 -0.6915 0.0747>}
 dyn_thigh_pack_l {parent=dyn_parent_thigh_pack_l, loc=<0 0 0>, rot=<0 0 0 -1>}
 thigh_r {parent=pelvis, loc=<-1.0853 -0.2109 8.7053>, rot=<0.0355 -0.0009 0.9993 0.0155>}
 calf_r {parent=thigh_r, loc=<42.5286 0.0019 0.7297>, rot=<0.0006 -0.0161 -0.0391 -0.9991>}
 calf_twist_01_r {parent=calf_r, loc=<35.8485 1.9197 0.5794>, rot=<0.0237 0.0095 -0.0035 -0.9997>}
 calf_twist_02_r {parent=calf_r, loc=<19.1546 1.7943 0.2638>, rot=<0.0237 0.0095 -0.0035 -0.9997>}
 foot_r {parent=calf_r, loc=<46.7403 2.5014 0.592>, rot=<0.0427 -0.0238 0.0353 -0.9982>}
 ball_r {parent=foot_r, loc=<3.5658 -10.8066 0.1544>, rot=<0.0001 -0.0001 -0.7063 -0.7079>}
 dyn_parent_shoe_tag_r {parent=foot_r, loc=<-6.4953 5.7041 -0.1041>, rot=<0.0001 -0.0454 0.9969 0.0639>}
 dyn_shoe_tag_01_r {parent=dyn_parent_shoe_tag_r, loc=<0 0 0>, rot=<0 0 0 -1>}
 thigh_twist_01_r {parent=thigh_r, loc=<27.952 0 0>, rot=<0.0013 0 0 -1>}
 sub_thigh_r {parent=thigh_r, loc=<2.9722 0 0>, rot=<0 0 0 -1>}
 driver_frill_frt_r {parent=sub_thigh_r, loc=<-6.7468 -8.5623 0.0995>, rot=<0 0 0.0087 -1>}
 dyn_frill_frt_r {parent=driver_frill_frt_r, loc=<0 0 0>, rot=<0 0 0 1>}
 driver_frill_side_r {parent=sub_thigh_r, loc=<-2.815 0.4685 8.1075>, rot=<0 0 0 -1>}
 dyn_frill_side_r {parent=driver_frill_side_r, loc=<0 0 0>, rot=<0 0 0 -1>}
 driver_frill_bck_r {parent=sub_thigh_r, loc=<-3.0964 9.5518 1.1757>, rot=<0 0 0 -1>}
 dyn_frill_bck_r {parent=driver_frill_bck_r, loc=<0 0 0>, rot=<0 0 0 -1>}
 dyn_tapedeck_01 {parent=pelvis, loc=<5.0665 6.0225 12.5621>, rot=<0.3301 0.0891 0.1775 -0.9228>}
 dyn_parent_acc_01 {parent=dyn_tapedeck_01, loc=<10.6016 -0.0949 -0.0548>, rot=<-0.4015 -0.6151 -0.4967 -0.4623>}
 dyn_driver_acc_01 {parent=dyn_parent_acc_01, loc=<0 0 0>, rot=<0 0 0 -1>}
 dyn_acc_01 {parent=dyn_driver_acc_01, loc=<0 0 0>, rot=<0 0 0 -1>}
 driver_frill_bck {parent=pelvis, loc=<-1.4605 -10.6717 0>, rot=<-0.0127 -0.707 0.0127 -0.707>}
 dyn_frill_bck {parent=driver_frill_bck, loc=<0 0 0>, rot=<0 0 0 1>}
 Flap_Front_Parent_Space {parent=pelvis, loc=<-0.8686 10.8513 0>, rot=<-0.0127 -0.707 0.0127 -0.707>}
 Flap_Front_Driven_Bone {parent=Flap_Front_Parent_Space, loc=<0 0 0>, rot=<0 0 0 1>}
 dyn_frill_frt {parent=Flap_Front_Driven_Bone, loc=<0 0 0>, rot=<0 0 0 1>}
]
