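# Physics asset for the skirt of the female medium Sailor Squad Rose player body, simulating its frill panels around pelvis and thigh shapes.
# Small driver bodies hang off the pelvis or sub-thigh bodies by free angular constraints, each anchoring a larger dynamic frill with swing and twist limits.
# Side and back frills link to neighbors, except right-back, which joins the right-side driver; the center-front flap has no neighbor links.
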PhysicsAsset /BRCosmetics/Characters/Player/Female/Medium/Bodies/F_MED_Sailor_Squad_Rose/Meshes/F_MED_Sailor_Squad_Rose_Skirt_Physics
bodies {
 Flap_Front_Driven_Bone [{shapeType=Sphere, radius=0.505}]
 driver_frill_bck [{shapeType=Sphere, radius=0.505}]
 driver_frill_bck_l [{shapeType=Sphere, radius=0.505}]
 driver_frill_bck_r [{shapeType=Sphere, radius=0.505}]
 driver_frill_frt_l [{shapeType=Sphere, radius=1}]
 driver_frill_frt_r [{shapeType=Sphere, radius=0.505}]
 driver_frill_side_l [{shapeType=Sphere, radius=0.505}]
 driver_frill_side_r [{shapeType=Sphere, radius=0.505}]
 dyn_frill_bck [{shapeType=Box, extentX=15, extentY=23, extentZ=6.5, center=<0 -0.8523 -11.7047>, rotation=<rot 0 0 90.7678>}]
 dyn_frill_bck_l [{shapeType=Box, extentX=23, extentY=14.5, extentZ=7.5, center=<-10 -1.25 -0.75>, rotation=<rot -4.411 -0.2685 66.5471>}]
 dyn_frill_bck_r [{shapeType=Box, extentX=23, extentY=14.5, extentZ=6, center=<10 1.25 0.75>, rotation=<rot -4.411 -0.2685 66.5471>}]
 dyn_frill_frt [{shapeType=Box, extentX=13.5, extentY=4, extentZ=20.5, center=<0 1.2344 -10.5551>, rotation=<rot 0 0 -3.1066>}]
 dyn_frill_frt_l [{shapeType=Capsule, length=17.5, radius=13, center=<-20.2582 10.4043 -4.7218>, rotation=<rot -64.9443 -93.5819 94.0161>}]
 dyn_frill_frt_r [{shapeType=Box, extentX=22, extentY=14, extentZ=5.5, center=<14.5 -1 1>, rotation=<rot -0.4202 -2.1687 -65.1905>}]
 dyn_frill_side_l [{shapeType=Box, extentX=21.5, extentY=17.5, extentZ=7.5, center=<-11 0 -1.5>, rotation=<rot -2.1878 -0.772 0.9964>}]
 dyn_frill_side_r [{shapeType=Box, extentX=21.5, extentY=15.5, extentZ=6.5, center=<11 0 1.5>, rotation=<rot -2.1878 -0.772 0.9964>}]
 pelvis [{shapeType=Capsule, length=10.8103, radius=8.1851, center=<-3.9396 0 0>}]
 sub_thigh_l [{shapeType=Capsule, length=0.505, radius=0.505, rotation=<rot 0 -90 180>}]
 sub_thigh_r [{shapeType=Capsule, length=0.505, radius=0.505, rotation=<rot 0 -90 0>}]
 thigh_l [{shapeType=Capsule, length=26.5, radius=9, center=<-18.1547 0.6081 0.1362>, rotation=<rot 89.761 0.0006 -179.9992>}]
 thigh_r [{shapeType=Capsule, length=19.5, radius=8.5, center=<18.1547 0.6081 0.1362>, rotation=<rot 89.761 0.0006 -179.9992>}]
}
constraints [
 {
  bone1Name=sub_thigh_r
  bone2Name=thigh_r
  swing1LimitDegrees=45
  swing1Motion=Free
  swing2LimitDegrees=45
  swing2Motion=Free
  twistLimitDegrees=45
  twistMotion=Free
 }
 {bone1Name=dyn_frill_bck, bone2Name=driver_frill_bck, swing1LimitDegrees=45, swing2LimitDegrees=45, twistLimitDegrees=10}
 {
  bone1Name=sub_thigh_l
  bone2Name=thigh_l
  swing1LimitDegrees=45
  swing1Motion=Free
  swing2LimitDegrees=45
  swing2Motion=Free
  twistLimitDegrees=45
  twistMotion=Free
 }
 {bone1Name=dyn_frill_side_l, bone2Name=driver_frill_side_l, swing1LimitDegrees=30, swing2LimitDegrees=30, twistLimitDegrees=10}
 {bone1Name=dyn_frill_frt_l, bone2Name=driver_frill_frt_l, swing1LimitDegrees=13, swing2LimitDegrees=20, twistLimitDegrees=10}
 {bone1Name=dyn_frill_bck_l, bone2Name=driver_frill_bck_l, swing1LimitDegrees=60, swing2LimitDegrees=45, twistLimitDegrees=15}
 {bone1Name=dyn_frill_bck, bone2Name=dyn_frill_bck_l, swing1LimitDegrees=45, swing2LimitDegrees=45, twistLimitDegrees=45}
 {bone1Name=dyn_frill_frt, bone2Name=Flap_Front_Driven_Bone, swing1LimitDegrees=45, swing2LimitDegrees=45, twistLimitDegrees=45}
 {
  bone1Name=thigh_r
  bone2Name=pelvis
  swing1LimitDegrees=45
  swing1Motion=Free
  swing2LimitDegrees=45
  swing2Motion=Free
  twistLimitDegrees=45
  twistMotion=Free
 }
 {
  bone1Name=driver_frill_bck
  bone2Name=pelvis
  swing1LimitDegrees=45
  swing1Motion=Free
  swing2LimitDegrees=45
  swing2Motion=Free
  twistLimitDegrees=45
  twistMotion=Free
 }
 {
  bone1Name=driver_frill_bck_r
  bone2Name=sub_thigh_r
  swing1LimitDegrees=45
  swing1Motion=Free
  swing2LimitDegrees=45
  swing2Motion=Free
  twistLimitDegrees=45
  twistMotion=Free
 }
 {
  bone1Name=driver_frill_side_r
  bone2Name=sub_thigh_r
  swing1LimitDegrees=45
  swing1Motion=Free
  swing2LimitDegrees=45
  swing2Motion=Free
  twistLimitDegrees=45
  twistMotion=Free
 }
 {
  bone1Name=driver_frill_frt_r
  bone2Name=sub_thigh_r
  swing1LimitDegrees=45
  swing1Motion=Free
  swing2LimitDegrees=45
  swing2Motion=Free
  twistLimitDegrees=45
  twistMotion=Free
 }
 {
  bone1Name=Flap_Front_Driven_Bone
  bone2Name=pelvis
  swing1LimitDegrees=45
  swing1Motion=Free
  swing2LimitDegrees=45
  swing2Motion=Free
  twistLimitDegrees=45
  twistMotion=Free
 }
 {
  bone1Name=driver_frill_bck_l
  bone2Name=sub_thigh_l
  swing1LimitDegrees=45
  swing1Motion=Free
  swing2LimitDegrees=45
  swing2Motion=Free
  twistLimitDegrees=45
  twistMotion=Free
 }
 {
  bone1Name=driver_frill_side_l
  bone2Name=sub_thigh_l
  swing1LimitDegrees=45
  swing1Motion=Free
  swing2LimitDegrees=45
  swing2Motion=Free
  twistLimitDegrees=45
  twistMotion=Free
 }
 {
  bone1Name=driver_frill_frt_l
  bone2Name=sub_thigh_l
  swing1LimitDegrees=45
  swing1Motion=Free
  swing2LimitDegrees=45
  swing2Motion=Free
  twistLimitDegrees=45
  twistMotion=Free
 }
 {bone1Name=thigh_l, bone2Name=pelvis, swing1LimitDegrees=45, swing2LimitDegrees=45, twistLimitDegrees=45}
 {bone1Name=dyn_frill_frt_r, bone2Name=driver_frill_frt_r, swing1LimitDegrees=13, swing2LimitDegrees=20, twistLimitDegrees=10}
 {bone1Name=dyn_frill_side_r, bone2Name=driver_frill_side_r, swing1LimitDegrees=30, swing2LimitDegrees=30, twistLimitDegrees=10}
 {bone1Name=dyn_frill_bck_r, bone2Name=driver_frill_bck_r, swing1LimitDegrees=60, swing2LimitDegrees=45, twistLimitDegrees=15}
 {bone1Name=dyn_frill_bck, bone2Name=dyn_frill_bck_r, swing1LimitDegrees=45, swing2LimitDegrees=45, twistLimitDegrees=45}
 {bone1Name=dyn_frill_side_r, bone2Name=dyn_frill_frt_r, swing1LimitDegrees=45, swing2LimitDegrees=45, twistLimitDegrees=45}
 {bone1Name=dyn_frill_side_l, bone2Name=dyn_frill_frt_l, swing1LimitDegrees=20, swing2LimitDegrees=20, twistLimitDegrees=20}
 {bone1Name=dyn_frill_bck_l, bone2Name=dyn_frill_side_l, swing1LimitDegrees=45, swing2LimitDegrees=45, twistLimitDegrees=45}
 {bone1Name=driver_frill_side_r, bone2Name=dyn_frill_bck_r, swing1LimitDegrees=45, swing2LimitDegrees=45, twistLimitDegrees=45}
]
