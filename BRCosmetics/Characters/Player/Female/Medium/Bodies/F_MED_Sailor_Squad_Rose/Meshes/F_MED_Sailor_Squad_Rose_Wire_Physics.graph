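# Physics asset for the wire, tape deck, zipper and shoe tag dangles on the female medium Sailor Squad Rose cosmetic body.
# A three-link coat wire hangs from spine_05 and joins the pelvis-mounted tape deck chain at dyn_acc_01, bridging both anchors in a closed loop.
# Zipper and shoe tags swing within tight limits from parent bodies, which attach to spine_02 and pelvis with free rotation.
# Capsules cover pelvis, spine and every dynamic bone.
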PhysicsAsset /BRCosmetics/Characters/Player/Female/Medium/Bodies/F_MED_Sailor_Squad_Rose/Meshes/F_MED_Sailor_Squad_Rose_Wire_Physics
bodies {
 dyn_acc_01 [{shapeType=Capsule, length=9.75, radius=3.75, center=<3.0295 5.6 0>, rotation=<rot 84.1733 0 -3.1045>}]
 dyn_driver_acc_01 [{shapeType=Capsule, length=0.505, radius=0.505, rotation=<rot 0 -90 0>}]
 dyn_driver_wire_coat_01 [{shapeType=Capsule, length=4.5, radius=4, rotation=<rot 90 0 0>}]
 dyn_parent_shoe_tag_l [{shapeType=Capsule, length=0.505, radius=0.505, rotation=<rot 0 -90 0>}]
 dyn_parent_shoe_tag_r [{shapeType=Capsule, length=0.505, radius=0.505, rotation=<rot 0 -90 0>}]
 dyn_parent_zipper [{shapeType=Capsule, length=0.505, radius=0.505, rotation=<rot 0 -90 0>}]
 dyn_shoe_tag_01_l [{shapeType=Capsule, length=3, radius=1.5, center=<1 0 0>, rotation=<rot 90 0 0>}]
 dyn_shoe_tag_01_r [{shapeType=Capsule, length=3, radius=1.5, center=<1 0 0>, rotation=<rot 90 0 0>}]
 dyn_tapedeck_01 [{shapeType=Capsule, length=5.199, radius=4.4996, center=<0.1 -0.3536 -0.0568>, rotation=<rot -1.3063 -74.5753 -87.2389>}]
 dyn_wire_coat_01 [{shapeType=Capsule, length=12.75, radius=2, center=<5 0 0>, rotation=<rot 90 0 0>}]
 dyn_wire_coat_02 [{shapeType=Capsule, length=13, radius=3, center=<2 4.3 0>, rotation=<rot 71.9021 179.2591 169.2711>}]
 dyn_wire_coat_03 [{shapeType=Capsule, length=9.75, radius=4, center=<1.65 2.8 -3.35>, rotation=<rot 48.7729 0 15.1631>}]
 dyn_zipper [{shapeType=Capsule, length=1.8875, radius=0.7538, center=<1.9997 0 0>, rotation=<rot 0 -98.0445 -86.8041>}]
 pelvis [{shapeType=Capsule, length=9.5, radius=12.75}]
 spine_02 [{shapeType=Capsule, length=5.25, radius=22, center=<-0.0149 -7.5676 -0.8141>, rotation=<rot -88.951 -179.9996 179.9996>}]
 spine_05 [{shapeType=Capsule, length=4, radius=18.25, center=<-6.2335 -6.5079 0>}]
}
constraints [
 {bone1Name=spine_05, bone2Name=spine_02, swing1LimitDegrees=45, swing2LimitDegrees=45, twistLimitDegrees=45}
 {
  bone1Name=spine_02
  bone2Name=pelvis
  swing1LimitDegrees=45
  swing1Motion=Free
  swing2LimitDegrees=45
  swing2Motion=Free
  twistLimitDegrees=45
  twistMotion=Free
 }
 {
  bone1Name=dyn_driver_wire_coat_01
  bone2Name=spine_05
  swing1LimitDegrees=45
  swing1Motion=Free
  swing2LimitDegrees=45
  swing2Motion=Free
  twistLimitDegrees=45
  twistMotion=Free
 }
 {bone1Name=dyn_wire_coat_01, bone2Name=dyn_driver_wire_coat_01, swing1LimitDegrees=10, swing2LimitDegrees=10, twistLimitDegrees=10}
 {bone1Name=dyn_wire_coat_02, bone2Name=dyn_wire_coat_01, swing1LimitDegrees=10, swing2LimitDegrees=10, twistLimitDegrees=10}
 {bone1Name=dyn_wire_coat_03, bone2Name=dyn_wire_coat_02, swing1LimitDegrees=15, swing2LimitDegrees=15, twistLimitDegrees=15}
 {bone1Name=dyn_driver_acc_01, bone2Name=dyn_tapedeck_01, swing1LimitDegrees=2, swing2LimitDegrees=2, twistLimitDegrees=2}
 {bone1Name=dyn_acc_01, bone2Name=dyn_driver_acc_01, swing1LimitDegrees=15, swing2LimitDegrees=15, twistLimitDegrees=5}
 {bone1Name=dyn_acc_01, bone2Name=dyn_wire_coat_03, swing1LimitDegrees=30, swing2LimitDegrees=30, twistLimitDegrees=30}
 {bone1Name=dyn_tapedeck_01, bone2Name=pelvis, swing1LimitDegrees=5, swing2LimitDegrees=5, twistLimitDegrees=3}
 {
  bone1Name=dyn_parent_zipper
  bone2Name=spine_02
  swing1LimitDegrees=45
  swing1Motion=Free
  swing2LimitDegrees=45
  swing2Motion=Free
  twistLimitDegrees=45
  twistMotion=Free
 }
 {bone1Name=dyn_zipper, bone2Name=dyn_parent_zipper, swing1LimitDegrees=15, swing2LimitDegrees=8, twistLimitDegrees=3}
 {
  bone1Name=dyn_parent_shoe_tag_l
  bone2Name=pelvis
  swing1LimitDegrees=45
  swing1Motion=Free
  swing2LimitDegrees=45
  swing2Motion=Free
  twistLimitDegrees=45
  twistMotion=Free
 }
 {bone1Name=dyn_shoe_tag_01_l, bone2Name=dyn_parent_shoe_tag_l, swing1LimitDegrees=15, swing2LimitDegrees=15, twistLimitDegrees=5}
 {
  bone1Name=dyn_parent_shoe_tag_r
  bone2Name=pelvis
  swing1LimitDegrees=45
  swing1Motion=Free
  swing2LimitDegrees=45
  swing2Motion=Free
  twistLimitDegrees=45
  twistMotion=Free
 }
 {bone1Name=dyn_shoe_tag_01_r, bone2Name=dyn_parent_shoe_tag_r, swing1LimitDegrees=15, swing2LimitDegrees=15, twistLimitDegrees=5}
]
