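# Post-process style AnimBP for the Sailor Squad Rose cosmetic body, animating its skirt frills, flap, thigh pouch, zipper and shoe tags.
# Copies the parent mesh pose through a Control Rig, then layers constraints, two RigidBody simulations and corrective bone offsets.
# Event graph derives corrective alphas from thigh-to-pelvis twist, crouch and skydive state.
# Cached result feeds a crouch/non-crouch offset blend; several computed variables, including scarf alphas, go unused here.
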AnimBlueprint /BRCosmetics/Characters/Player/Female/Medium/Bodies/F_MED_Sailor_Squad_Rose/Meshes/F_MED_Sailor_Squad_Rose_AnimBP
target_skeleton=&/F_MED_Sailor_Squad_Rose_Skeleton
graphs [
 AnimGraph {nodes [
   n1 = Get legUp_R() -> (legUp_R:double)
   n2 = "Transform (Modify) Bone - Bone: driver_frill_side_r"(Alpha=@n1, Translation="0,0,2", Rotation="0,0,0", Scale="1,1,1") {
    rotationMode=BMM_Additive
    rotationSpace=BCS_BoneSpace
    translationMode=BMM_Additive
    translationSpace=BCS_BoneSpace
    alphaBoolBlend {blendOption=Linear}
    alphaScaleBiasClamp {bMapRange=true, inRange {max=0, min=-100}, outRange {max=0, min=1}}
    boneToModify {boneName=driver_frill_side_r}
    translation {z=2}
   }
   "Transform (Modify) Bone - Bone: driver_frill_bck_r"(ComponentPose=@n2, Alpha=@n1, Translation="-5,0,0", Rotation="0,80,0", Scale="1,1,1") {
    rotationMode=BMM_Additive
    rotationSpace=BCS_BoneSpace
    translationMode=BMM_Additive
    translationSpace=BCS_BoneSpace
    alphaBoolBlend {blendOption=Linear}
    alphaScaleBiasClamp {bMapRange=true, inRange {max=0, min=-100}, outRange {max=0, min=1}}
    boneToModify {boneName=driver_frill_bck_r}
    rotation {yaw=80}
    translation {x=-5}
   }
   "Transform (Modify) Bone - Bone: driver_frill_bck_l"(Alpha=Get legUp_L() -> (legUp_L:double), Translation="5,0,0", Rotation="0,80,0", Scale="1,1,1") {
    rotationMode=BMM_Additive
    rotationSpace=BCS_BoneSpace
    translationMode=BMM_Additive
    translationSpace=BCS_BoneSpace
    alphaBoolBlend {blendOption=Linear}
    alphaScaleBiasClamp {bMapRange=true, inRange {max=0, min=-100}, outRange {max=0, min=1}}
    boneToModify {boneName=driver_frill_bck_l}
    rotation {yaw=80}
    translation {x=5}
   }
   n3 = "Transform (Modify) Bone - Bone: Flap_Front_Parent_Space"(
    ComponentPose=Local To Component(LocalPose=Use cached pose 'PreCrouch'())
    Alpha=1
    Translation="0,0,0"
    Rotation="0.000,0,-55"
    Scale="1,1,1"
   ) {
    rotationMode=BMM_Additive
    rotationSpace=BCS_BoneSpace
    alphaBoolBlend {blendOption=Linear}
    boneToModify {boneName=Flap_Front_Parent_Space}
    rotation {roll=-55}
   }
   n4 = "Transform (Modify) Bone - Bone: driver_frill_side_l"(ComponentPose=@n3, Alpha=1, Translation="0,1,-1", Rotation="5,-10,0.000", Scale="1,1,1") {
    rotationMode=BMM_Additive
    rotationSpace=BCS_BoneSpace
    translationMode=BMM_Additive
    translationSpace=BCS_BoneSpace
    alphaBoolBlend {blendOption=Linear}
    alphaScaleBiasClamp {bInterpResult=true, interpSpeedDecreasing=15, interpSpeedIncreasing=15}
    boneToModify {boneName=driver_frill_side_l}
    rotation {pitch=5, yaw=-10}
    translation {y=1, z=-1}
   }
   n5 = Component To Local(ComponentPose=@n4)
   n6 = "Transform (Modify) Bone - Bone: Flap_Front_Driven_Bone"(
    ComponentPose=Local To Component(LocalPose=Use cached pose 'PreCrouch'())
    Alpha=1
    Translation="0,0,0"
    Rotation="32.680,26.356,-52.2"
    Scale="1,1,1"
   ) {
    rotationMode=BMM_Additive
    rotationSpace=BCS_BoneSpace
    translationMode=BMM_Additive
    translationSpace=BCS_BoneSpace
    alphaBoolBlend {blendOption=Linear}
    alphaScaleBiasClamp {bInterpResult=true}
    boneToModify {boneName=Flap_Front_Driven_Bone}
    rotation=<rot 32.68 26.356 -52.2>
   }
   n7 = "Transform (Modify) Bone - Bone: driver_frill_frt_l"(ComponentPose=@n6, Alpha=1, Translation="-3,-3,-5", Rotation="0,-8,-10", Scale="1,1,1") {
    rotationMode=BMM_Additive
    rotationSpace=BCS_BoneSpace
    translationMode=BMM_Additive
    translationSpace=BCS_BoneSpace
    alphaBoolBlend {blendOption=Linear}
    alphaScaleBiasClamp {interpSpeedDecreasing=15, interpSpeedIncreasing=15}
    boneToModify {boneName=driver_frill_frt_l}
    rotation {roll=-10, yaw=-8}
    translation=<-3 -3 -5>
   }
   n8 = "Transform (Modify) Bone - Bone: dyn_frill_side_l"(ComponentPose=@n7, Alpha=1, Translation="-5,0,-5", Rotation="0,0,0", Scale="1,1,1") {
    rotationMode=BMM_Additive
    rotationSpace=BCS_BoneSpace
    translationMode=BMM_Additive
    translationSpace=BCS_BoneSpace
    alphaBoolBlend {blendOption=Linear}
    alphaScaleBiasClamp {interpSpeedDecreasing=15, interpSpeedIncreasing=15}
    boneToModify {boneName=dyn_frill_side_l}
    translation {x=-5, z=-5}
   }
   n9 = Get crouch_legUp_R() -> (crouch_legUp_R:double)
   n10 = "Transform (Modify) Bone - Bone: dyn_frill_frt_r"(ComponentPose=@n8, Alpha=@n9, Translation="0,-1,0", Rotation="0,-3,0", Scale="1,1,1") {
    rotationMode=BMM_Additive
    rotationSpace=BCS_BoneSpace
    translationMode=BMM_Additive
    translationSpace=BCS_BoneSpace
    alphaBoolBlend {blendOption=Linear}
    alphaScaleBiasClamp {bInterpResult=true, interpSpeedDecreasing=15, interpSpeedIncreasing=15}
    boneToModify {boneName=dyn_frill_frt_r}
    rotation {yaw=-3}
    translation {y=-1}
   }
   n11 = "Transform (Modify) Bone - Bone: dyn_frill_side_r"(ComponentPose=@n10, Alpha=@n9, Translation="0,-2,2", Rotation="0,0,0", Scale="1,1,1") {
    rotationMode=BMM_Additive
    rotationSpace=BCS_BoneSpace
    translationMode=BMM_Additive
    translationSpace=BCS_BoneSpace
    alphaBoolBlend {blendOption=Linear}
    alphaScaleBiasClamp {bInterpResult=true, interpSpeedDecreasing=15, interpSpeedIncreasing=15}
    boneToModify {boneName=dyn_frill_side_r}
    translation {y=-2, z=2}
   }
   n12 = "Transform (Modify) Bone - Bone: driver_frill_side_l"(ComponentPose=@n11, Alpha=1, Translation="1,10,0", Rotation="-15,0,20", Scale="1,1,1") {
    rotationMode=BMM_Additive
    rotationSpace=BCS_BoneSpace
    translationMode=BMM_Additive
    translationSpace=BCS_BoneSpace
    alphaBoolBlend {blendOption=Linear}
    alphaScaleBiasClamp {bInterpResult=true, interpSpeedDecreasing=15, interpSpeedIncreasing=15}
    boneToModify {boneName=driver_frill_side_l}
    rotation {pitch=-15, roll=20}
    translation=<1 10>
   }
   n13 = "Transform (Modify) Bone - Bone: driver_frill_frt_l"(ComponentPose=@n12, Alpha=1, Translation="-7,3,-2", Rotation="0,7,0", Scale="0.75,0.75,0.75") {
    rotationMode=BMM_Additive
    rotationSpace=BCS_BoneSpace
    scaleMode=BMM_Additive
    scaleSpace=BCS_BoneSpace
    translationMode=BMM_Additive
    translationSpace=BCS_BoneSpace
    alphaBoolBlend {blendOption=Linear}
    alphaScaleBiasClamp {bInterpResult=true, interpSpeedDecreasing=15, interpSpeedIncreasing=15}
    boneToModify {boneName=driver_frill_frt_l}
    rotation {yaw=7}
    scale=<0.75 0.75 0.75>
    translation=<-7 3 -2>
   }
   n14 = "Transform (Modify) Bone - Bone: driver_frill_side_l"(ComponentPose=@n13, Alpha=Get legOut_L() -> (legOut_L:double), Translation="0,0,-2", Rotation="0,0,0", Scale="1,1,1") {
    rotationMode=BMM_Additive
    rotationSpace=BCS_BoneSpace
    translationMode=BMM_Additive
    translationSpace=BCS_BoneSpace
    alphaBoolBlend {blendOption=Linear}
    alphaScaleBiasClamp {bInterpResult=true, interpSpeedDecreasing=15, interpSpeedIncreasing=15}
    boneToModify {boneName=driver_frill_side_l}
    translation {z=-2}
   }
   n15 = "Transform (Modify) Bone - Bone: driver_frill_side_l"(ComponentPose=@n14, Alpha=1, Translation="0,-5,0", Rotation="15,-15,0.000", Scale="1,1,1") {
    rotationMode=BMM_Additive
    rotationSpace=BCS_BoneSpace
    translationMode=BMM_Additive
    translationSpace=BCS_BoneSpace
    alphaBoolBlend {blendOption=Linear}
    alphaScaleBiasClamp {bInterpResult=true, interpSpeedDecreasing=15, interpSpeedIncreasing=15}
    boneToModify {boneName=driver_frill_side_l}
    rotation {pitch=15, yaw=-15}
    translation {y=-5}
   }
   n16 = Component To Local(ComponentPose=@n15)
   n17 = Get bIsCrouching()
   n18 = "Transform (Modify) Bone - Bone: driver_frill_side_l"(
    ComponentPose=Local To Component(LocalPose=Blend Poses by bool(BlendPose_0=@n16, BlendPose_1=@n5, bActiveValue=@n17) {blendType=Linear})
    Alpha=1
    Translation="-5,0.000,-2"
    Rotation="10,0,0"
    Scale="1,1,1"
   ) {
    rotationMode=BMM_Additive
    rotationSpace=BCS_BoneSpace
    translationMode=BMM_Additive
    translationSpace=BCS_BoneSpace
    alphaBoolBlend {blendOption=Linear}
    boneToModify {boneName=driver_frill_side_l}
    rotation {pitch=10}
    translation {x=-5, z=-2}
   }
   Output Pose(Result=Component To Local(ComponentPose=@n18)) {layerGroup=None}
   n19 = Control Rig(Source=Copy Pose From Mesh() {bUseAttachedParent=true}) {
    alphaBoolBlend {blendOption=Linear}
    controlRigAssetReference {blueprintRigClass=&/F_MED_Sailor_Squad_Rose_CtrlRig.F_MED_Sailor_Squad_Rose_CtrlRig_C}
   }
   n20 = "Transform (Modify) Bone - Bone: pelvis_alt"(ComponentPose=Local To Component(LocalPose=@n19), Alpha=1, Translation="0,4,0", Rotation="0,0,0", Scale="1,1,1") {
    translationMode=BMM_Additive
    translationSpace=BCS_BoneSpace
    alphaBoolBlend {blendOption=Linear}
    boneToModify {boneName=pelvis_alt}
    translation {y=4}
   }
   n21 = "Transform (Modify) Bone - Bone: spine_01_alt"(ComponentPose=@n20, Alpha=1, Translation="0,4,0", Rotation="0,0,0", Scale="1,1,1") {
    translationMode=BMM_Additive
    translationSpace=BCS_BoneSpace
    alphaBoolBlend {blendOption=Linear}
    boneToModify {boneName=spine_01_alt}
    translation {y=4}
   }
   n22 = "Transform (Modify) Bone - Bone: spine_02_alt"(ComponentPose=@n21, Alpha=1, Translation="0,4,0", Rotation="0,0,0", Scale="1,1,1") {
    translationMode=BMM_Additive
    translationSpace=BCS_BoneSpace
    alphaBoolBlend {blendOption=Linear}
    boneToModify {boneName=spine_02_alt}
    translation {y=4}
   }
   n23 = "Transform (Modify) Bone - Bone: spine_03_alt"(ComponentPose=@n22, Alpha=1, Translation="0,4,0", Rotation="0,0,0", Scale="1,1,1") {
    translationMode=BMM_Additive
    translationSpace=BCS_BoneSpace
    alphaBoolBlend {blendOption=Linear}
    boneToModify {boneName=spine_03_alt}
    translation {y=4}
   }
   n24 = "Transform (Modify) Bone - Bone: spine_04_alt"(ComponentPose=@n23, Alpha=1, Translation="0,4,0", Rotation="0,0,0", Scale="1,1,1") {
    translationMode=BMM_Additive
    translationSpace=BCS_BoneSpace
    alphaBoolBlend {blendOption=Linear}
    boneToModify {boneName=spine_04_alt}
    translation {y=4}
   }
   n25 = "Constraint - dyn_parent_zipper (PPPPPP)"!(
    ComponentPose=@n24
    Alpha=1
    ConstraintWeights_0=0.0887
    ConstraintWeights_1:float=0.0143
    ConstraintWeights_2:float=0.0206
    ConstraintWeights_3:float=0.5625
    ConstraintWeights_4:float=0.181
    ConstraintWeights_5:float=0.1329
   ) -> (Pose) {
    alphaBoolBlend {blendOption=Linear}
    boneToModify {boneName=dyn_parent_zipper}
    constraintSetup [
     {targetBone {boneName=upperarm_twist_01_r}}
     {targetBone {boneName=spine_02}}
     {targetBone {boneName=spine_03}}
     {targetBone {boneName=spine_04}}
     {targetBone {boneName=spine_05}}
     {targetBone {boneName=clavicle_r}}
    ]
    constraintWeights [1,1,1,1,1,1]
   }
   n26 = "Constraint - dyn_parent_shoe_tag_l (PPPP)"!(
    ComponentPose=@n25
    Alpha=1
    ConstraintWeights_0=0.0211
    ConstraintWeights_1:float=0.6538
    ConstraintWeights_2:float=0.2835
    ConstraintWeights_3:float=0.0416
   ) -> (Pose) {
    alphaBoolBlend {blendOption=Linear}
    boneToModify {boneName=dyn_parent_shoe_tag_l}
    constraintSetup [
     {targetBone {boneName=calf_l}}
     {targetBone {boneName=calf_twist_01_l}}
     {targetBone {boneName=foot_l}}
     {targetBone {boneName=calf_twist_02_l}}
    ]
    constraintWeights [1,1,1,1]
   }
   n27 = "Constraint - dyn_parent_shoe_tag_r (PPPP)"!(
    ComponentPose=@n26
    Alpha=1
    ConstraintWeights_0=0.0416
    ConstraintWeights_1:float=0.6538
    ConstraintWeights_2:float=0.2835
    ConstraintWeights_3:float=0.0211
   ) -> (Pose) {
    alphaBoolBlend {blendOption=Linear}
    alphaScaleBiasClamp {bInterpResult=true, interpSpeedDecreasing=15, interpSpeedIncreasing=15}
    boneToModify {boneName=dyn_parent_shoe_tag_r}
    constraintSetup [
     {targetBone {boneName=calf_twist_02_r}}
     {targetBone {boneName=calf_twist_01_r}}
     {targetBone {boneName=foot_r}}
     {targetBone {boneName=calf_r}}
    ]
    constraintWeights [1,1,1,1]
   }
   n28 = RigidBody(ComponentPose=@n27) {
    bOverrideWorldGravity=true
    overridePhysicsAsset=&/F_MED_Sailor_Squad_Rose_Wire_Physics
    simulationSpace=BaseBoneSpace
    alphaBoolBlend {blendOption=Linear}
    baseBoneRef {boneName=root}
    componentAppliedLinearAccClamp=<1100 1100 1000>
    componentLinearAccScale=<0.5 0.5 0.5>
    componentLinearVelScale=<0.5 0.5 0.5>
    overrideWorldGravity {z=-980}
   }
   n29 = "Constraint - dyn_parent_thigh_pack_l (P)"!(ComponentPose=@n28, Alpha=1, ConstraintWeights_0=1, ConstraintWeights_1:float=0) -> (Pose) {
    alphaBoolBlend {blendOption=Linear}
    boneToModify {boneName=dyn_parent_thigh_pack_l}
    constraintSetup [{targetBone {boneName=thigh_l}}]
    constraintWeights [1,0]
   }
   n30 = "Transform (Modify) Bone - Bone: dyn_thigh_pack_l"(ComponentPose=@n29, Alpha=Get crouch_pouch() -> (crouch_pouch:double), Translation="-1,-0.000,2", Rotation="-3.5,0.25,15", Scale="1,1,1") {
    rotationMode=BMM_Additive
    rotationSpace=BCS_BoneSpace
    translationMode=BMM_Additive
    translationSpace=BCS_BoneSpace
    alphaBoolBlend {blendOption=Linear}
    boneToModify {boneName=dyn_thigh_pack_l}
    rotation=<rot -3.5 0.25 15>
    translation {x=-1, z=2}
   }
   n31 = "Constraint - sub_thigh_l (P)"(ComponentPose=@n30, Alpha=1, ConstraintWeights_0=1) {
    alphaBoolBlend {blendOption=Linear}
    boneToModify {boneName=sub_thigh_l}
    constraintSetup [{targetBone {boneName=thigh_twist_01_l}}]
    constraintWeights [1]
   }
   n32 = "Constraint - sub_thigh_r (P)"(ComponentPose=@n31, Alpha=1, ConstraintWeights_0=1) {
    alphaBoolBlend {blendOption=Linear}
    boneToModify {boneName=sub_thigh_r}
    constraintSetup [{targetBone {boneName=thigh_twist_01_r}}]
    constraintWeights [1]
   }
   n33 = Get legUp_L()
   n34 = "Transform (Modify) Bone - Bone: driver_frill_frt_l"(ComponentPose=@n32, Alpha=@n33, Translation="-4,20,2.5", Rotation="-18,35,10", Scale="1,1,1") {
    rotationMode=BMM_Additive
    rotationSpace=BCS_BoneSpace
    translationMode=BMM_Additive
    translationSpace=BCS_BoneSpace
    alphaBoolBlend {blendOption=Linear}
    alphaScaleBiasClamp {bInterpResult=true, interpSpeedDecreasing=15, interpSpeedIncreasing=15}
    boneToModify {boneName=driver_frill_frt_l}
    rotation=<rot -18 35 10>
    translation=<-4 20 2.5>
   }
   n35 = "Transform (Modify) Bone - Bone: driver_frill_side_l"(ComponentPose=@n34, Alpha=@n33, Translation="5,0,0", Rotation="-15,0,0", Scale="1,1,1") {
    rotationMode=BMM_Additive
    rotationSpace=BCS_BoneSpace
    scaleMode=BMM_Additive
    scaleSpace=BCS_BoneSpace
    translationMode=BMM_Additive
    translationSpace=BCS_BoneSpace
    alphaBoolBlend {blendOption=Linear}
    alphaScaleBiasClamp {bInterpResult=true, interpSpeedDecreasing=15, interpSpeedIncreasing=15, inRange {max=0, min=-100}, outRange {max=0, min=1}}
    boneToModify {boneName=driver_frill_side_l}
    rotation {pitch=-15}
    translation {x=5}
   }
   n36 = "Transform (Modify) Bone - Bone: driver_frill_frt_r"(ComponentPose=@n35, Alpha=Get legUp_R(), Translation="-10,-4.25,-1", Rotation="0,0,0", Scale="1,1,1") {
    rotationMode=BMM_Additive
    rotationSpace=BCS_BoneSpace
    translationMode=BMM_Additive
    translationSpace=BCS_BoneSpace
    alphaBoolBlend {blendOption=Linear}
    alphaScaleBiasClamp {bInterpResult=true, interpSpeedDecreasing=15, interpSpeedIncreasing=15, inRange {max=0, min=-100}, outRange {max=0, min=1}}
    boneToModify {boneName=driver_frill_frt_r}
    translation=<-10 -4.25 -1>
   }
   n37 = "Constraint - Flap_Front_Driven_Bone (PP)"!(ComponentPose=@n36, Alpha=1, ConstraintWeights_0=0.5, ConstraintWeights_1:float=0.5) -> (Pose) {
    alphaBoolBlend {blendOption=Linear}
    boneToModify {boneName=Flap_Front_Driven_Bone}
    constraintSetup [{targetBone {boneName=dyn_frill_frt_l}}, {targetBone {boneName=dyn_frill_frt_r}}]
    constraintWeights [1,1]
   }
   n38 = "Constraint - driver_frill_bck (PP)"!(ComponentPose=@n37, Alpha=1, ConstraintWeights_0=0.5, ConstraintWeights_1:float=0.5) -> (Pose) {
    alphaBoolBlend {blendOption=Linear}
    boneToModify {boneName=driver_frill_bck}
    constraintSetup [{targetBone {boneName=dyn_frill_bck_l}}, {targetBone {boneName=dyn_frill_bck_r}}]
    constraintWeights [1,1]
   }
   n39 = "Transform (Modify) Bone - Bone: driver_frill_side_l"(ComponentPose=@n38, Alpha=Get legOut_L(), Translation="0,0,-2", Rotation="8,8,0", Scale="1,1,1") {
    rotationMode=BMM_Additive
    rotationSpace=BCS_BoneSpace
    translationMode=BMM_Additive
    translationSpace=BCS_BoneSpace
    alphaBoolBlend {blendOption=Linear}
    alphaScaleBiasClamp {bInterpResult=true, interpSpeedDecreasing=30, interpSpeedIncreasing=30}
    boneToModify {boneName=driver_frill_side_l}
    rotation {pitch=8, yaw=8}
    translation {z=-2}
   }
   n40 = RigidBody(ComponentPose=@n39) {
    bClampLinearTranslationLimitToRefPose=true
    bOverrideWorldGravity=true
    overridePhysicsAsset=&/F_MED_Sailor_Squad_Rose_Skirt_Physics
    simulationSpace=BaseBoneSpace
    alphaBoolBlend {blendOption=Linear}
    baseBoneRef {boneName=root}
    componentLinearAccScale=<0.5 0.5 0.35>
    componentLinearVelScale=<0.5 0.5 0.35>
    overrideWorldGravity {z=-980}
   }
   n41 = Get isSkyDivingOn() -> (isSkyDivingOn:double)
   n42 = "Transform (Modify) Bone - Bone: dyn_frill_bck"(ComponentPose=@n40, Alpha=@n41, Translation="0,-3,0", Rotation="0,0,0", Scale="1,1,1") {
    rotationMode=BMM_Additive
    rotationSpace=BCS_BoneSpace
    translationMode=BMM_Additive
    translationSpace=BCS_BoneSpace
    alphaBoolBlend {blendOption=Linear}
    alphaScaleBiasClamp {bInterpResult=true, outRange {max=0, min=1}}
    boneToModify {boneName=dyn_frill_bck}
    translation {y=-3}
   }
   n43 = "Transform (Modify) Bone - Bone: dyn_frill_bck_l"(ComponentPose=@n42, Alpha=@n41, Translation="0,-3,0", Rotation="0,0,0", Scale="1,1,1") {
    rotationMode=BMM_Additive
    rotationSpace=BCS_BoneSpace
    translationMode=BMM_Additive
    translationSpace=BCS_BoneSpace
    alphaBoolBlend {blendOption=Linear}
    alphaScaleBiasClamp {bInterpResult=true, outRange {max=0, min=1}}
    boneToModify {boneName=dyn_frill_bck_l}
    translation {y=-3}
   }
   n44 = "Transform (Modify) Bone - Bone: dyn_frill_bck_r"(ComponentPose=@n43, Alpha=@n41, Translation="0,3,0", Rotation="0,0,0", Scale="1,1,1") {
    rotationMode=BMM_Additive
    rotationSpace=BCS_BoneSpace
    translationMode=BMM_Additive
    translationSpace=BCS_BoneSpace
    alphaBoolBlend {blendOption=Linear}
    alphaScaleBiasClamp {bInterpResult=true, outRange {max=0, min=1}}
    boneToModify {boneName=dyn_frill_bck_r}
    translation {y=3}
   }
   n45 = "Transform (Modify) Bone - Bone: dyn_frill_frt_r"(ComponentPose=@n44, Alpha=Get crouch_legUp_R(), Translation="5,3,0", Rotation="0,10,0", Scale="1,1,1") {
    rotationMode=BMM_Additive
    rotationSpace=BCS_BoneSpace
    translationMode=BMM_Additive
    translationSpace=BCS_BoneSpace
    alphaBoolBlend {blendOption=Linear}
    alphaScaleBiasClamp {bInterpResult=true, interpSpeedDecreasing=15, interpSpeedIncreasing=15}
    boneToModify {boneName=dyn_frill_frt_r}
    rotation {yaw=10}
    translation=<5 3>
   }
   n46 = Get legOut_R() -> (legOut_R:double)
   n47 = "Transform (Modify) Bone - Bone: dyn_frill_side_r"(ComponentPose=@n45, Alpha=@n46, Translation="0,0,1", Rotation="0,0,0", Scale="1,1,1") {
    rotationMode=BMM_Additive
    rotationSpace=BCS_BoneSpace
    translationMode=BMM_Additive
    translationSpace=BCS_BoneSpace
    alphaBoolBlend {blendOption=Linear}
    alphaScaleBiasClamp {bInterpResult=true, interpSpeedDecreasing=15, interpSpeedIncreasing=15}
    boneToModify {boneName=dyn_frill_side_r}
    translation {z=1}
   }
   n48 = "Transform (Modify) Bone - Bone: driver_frill_frt_r"(ComponentPose=@n47, Alpha=@n46, Translation="0,1,0", Rotation="0,3,0", Scale="1,1,1") {
    rotationMode=BMM_Additive
    rotationSpace=BCS_BoneSpace
    translationMode=BMM_Additive
    translationSpace=BCS_BoneSpace
    alphaBoolBlend {blendOption=Linear}
    alphaScaleBiasClamp {bInterpResult=true, interpSpeedDecreasing=15, interpSpeedIncreasing=15}
    boneToModify {boneName=driver_frill_frt_r}
    rotation {yaw=3}
    translation {y=1}
   }
   n49 = "Transform (Modify) Bone - Bone: driver_frill_side_l"(ComponentPose=@n48, Alpha=Get legOut_L(), Translation="0,0,-2", Rotation="0,0,0", Scale="1,1,1") {
    rotationMode=BMM_Additive
    rotationSpace=BCS_BoneSpace
    translationMode=BMM_Additive
    translationSpace=BCS_BoneSpace
    alphaBoolBlend {blendOption=Linear}
    alphaScaleBiasClamp {bInterpResult=true, interpSpeedDecreasing=15, interpSpeedIncreasing=15}
    boneToModify {boneName=driver_frill_side_l}
    translation {z=-2}
   }
   Save cached pose 'PreCrouch'(Pose=Component To Local(ComponentPose=@n49)) {cacheName=PreCrouch}
   Get legIn_L() -> (legIn_L:double)
  ]}
 EventGraph {nodes [
   Event OnCharacterCustomizationLoaded()
   > Do Once()
   Event BlueprintUpdateAnimation()
   > Get!(execute:Exec) -> (then:Exec, else:Exec, Body_Mesh:Skeletal Mesh Component Object Reference)
   > n1 = Sequence!(execute) -> (then_0, then_1, then_2:Exec, then_3:Exec, then_4:Exec)
   > n2 = Branch(Condition=Get bIsCrouching())
   n3 = Get PawnSpeed2D()
   > n4 = Branch(Condition="float <= float"!(A:double=MapRangeClamped(Value=@n3, InRangeA=0, InRangeB=150), B:double=0) -> (ReturnValue:bool) {is=title K2Node_CallFunction})
   > Set crouch_speed_alpha(crouch_speed_alpha:double=1) -> (Output_Get:double)
   > Set crouch_moving(crouch_moving:bool=false) -> (Output_Get:bool)
   > Set crouch_rigid_alpha(crouch_rigid_alpha:double=0.2) -> (Output_Get:double)
   > Set move_speed_alpha(move_speed_alpha:double=0) -> (Output_Get:double)
   Set crouch_speed_alpha(execute=@n4.else, crouch_speed_alpha=0)
   > Set crouch_moving(crouch_moving=true)
   > Set crouch_rigid_alpha(crouch_rigid_alpha=0.55)
   > Set move_speed_alpha(move_speed_alpha=1)
   Set crouch_rigid_alpha(execute=@n2.else, crouch_rigid_alpha=0.8)
   n5 = CheckForBackbling()
   Set backbling_active(execute=@n1.then_3, backbling_active:bool=OR Boolean(A=@n5, B=@n5.bIsBackblingLong)) -> (Output_Get:bool)
   > Set isWearingCape(isWearingCape=@n5.bIsBackblingACape)
   n6 = Sequence(execute=@n1.then_4)
   n7 = TryGetFortPawn()
   Is Valid(exec=@n6.then_2, InputObject=@n7)
   n8 = Get bIsDBNO!(self:Fort Pawn Object Reference=@n7) -> (bIsDBNO)
   > Set IsDBNO(IsDBNO:bool=@n8) -> (Output_Get:bool)
   > Set NotDBNO(NotDBNO:bool=NOT Boolean(A=@n8)) -> (Output_Get:bool)
   > n9 = Sequence!(execute) -> (then_0, then_1)
   n10 = Get bIsCrouching()
   > Set crouchAlpha(crouchAlpha:double="To Float (Boolean)"(InBool=@n10)) -> (Output_Get:double)
   n11 = Branch(execute=@n9.then_1, Condition=@n10)
   n12 = Get PawnSpeed2D()
   > n13 = Branch(Condition="float <= float"(A=MapRangeClamped(Value=@n12, InRangeA=0, InRangeB=150), B=0))
   > Set crouch_speed_alpha(crouch_speed_alpha=1)
   > Set crouch_moving(crouch_moving=false)
   > Set crouch_rigid_alpha(crouch_rigid_alpha=0.2)
   > Set move_speed_alpha(move_speed_alpha=0)
   > Set crouch_pouch(crouch_pouch:double=1) -> (Output_Get:double)
   Set crouch_speed_alpha(execute=@n13.else, crouch_speed_alpha=0)
   > Set crouch_moving(crouch_moving=true)
   > Set crouch_rigid_alpha(crouch_rigid_alpha=0.55)
   > Set move_speed_alpha(move_speed_alpha=1)
   > Set crouch_pouch(crouch_pouch=1)
   Set crouch_rigid_alpha(execute=@n11.else, crouch_rigid_alpha=0.8)
   n14 = GetOwnerDirection!(self:Custom Character Part Anim Instance Object Reference) -> (ReturnValue:float)
   n15 = Branch(
    execute=@n6.then_0
    Condition=AND Boolean(
     A="float > float"(A=@n14, B=-120) {is=title K2Node_CallFunction}
     B="float < float"!(A:double=@n14, B:double=-60) -> (ReturnValue:bool) {is=title K2Node_CallFunction}
    )
   )
   n16 = Get Body_Mesh()
   n17 = BreakTransform(InTransform=GetDeltaTransformFromRefPose(self=@n16, BoneName=thigh_r, BaseName=pelvis))
   > Set moveLeftLegCrossAlpha(moveLeftLegCrossAlpha:double=MapRangeClamped(Value=BreakRotator(InRot=@n17.Rotation).Pitch, InRangeA=-15, InRangeB=0, OutRangeA=1, OutRangeB=0)) -> (Output_Get:double)
   Set moveLeftLegCrossAlpha(execute=@n15.else, moveLeftLegCrossAlpha=0)
   n18 = Branch(execute=@n6.then_1, Condition=AND Boolean(A="float > float"(A=@n14, B=60), B="float < float"(A=@n14, B=120)))
   n19 = BreakTransform(InTransform=GetDeltaTransformFromRefPose(self=@n16, BoneName=thigh_l, BaseName=pelvis))
   > Set moveRighttLegCrossAlpha(moveRighttLegCrossAlpha:double=MapRangeClamped(Value=BreakRotator(InRot=@n19.Rotation).Pitch, InRangeA=-15, InRangeB=0, OutRangeA=1, OutRangeB=0)) -> (Output_Get:double)
   Set moveRighttLegCrossAlpha(execute=@n18.else, moveRighttLegCrossAlpha=0)
   n20 = "Absolute (Float)"(A=GetOwnerDirection())
   Set isMovingSide(execute=@n6.then_4, isMovingSide:bool=AND Boolean(A="float > float"(A=@n20, B=60), B="float < float"(A=@n20, B=120))) -> (Output_Get:bool)
   > n21 = Branch(Condition=Get bRigidBodyEnabled())
   > Set not_rigids_active(not_rigids_active:bool=false) -> (Output_Get:bool)
   Set not_rigids_active(execute=@n21.else, not_rigids_active=true)
   Set jump_alpha(execute=@n6.then_3, jump_alpha:double=MapRangeClamped(Value=Get PawnVelocityZ(), InRangeA=-700, InRangeB=0, OutRangeA=1, OutRangeB=0)) -> (Output_Get:double)
   > n22 = Branch()
   > Set isSkyDivingOn(isSkyDivingOn:double="To Float (Boolean)"(InBool=Get bIsSkyDiving())) -> (Output_Get:double)
   Set isSkyDivingOn(execute=@n22.else, isSkyDivingOn=0)
   Set legUp_L(
    execute=@n1.then_2
    legUp_L:double=GetAlphaFromReferenceTransformDelta(
     SkeletalMeshComponent=Get Body_Mesh()
     BoneName=thigh_l
     BaseName=pelvis
     DeltaComponentType=QuaternionTwist
     InputRange=(X=25,Y=95)
     TwistAxis="0,0,1"
    )
   ) -> (Output_Get:double)
   > Set legUp_R(legUp_R:double=GetAlphaFromReferenceTransformDelta(
     SkeletalMeshComponent=Get Body_Mesh()
     BoneName=thigh_r
     BaseName=pelvis
     DeltaComponentType=QuaternionTwist
     InputRange=(X=25,Y=95)
     TwistAxis="0,0,1"
    )) -> (Output_Get:double)
   > Set legOut_L(legOut_L:double=GetAlphaFromReferenceTransformDelta(
     SkeletalMeshComponent=Get Body_Mesh()
     BoneName=thigh_l
     BaseName=pelvis
     DeltaComponentType=QuaternionTwist
     InputRange=(X=0,Y=30)
     TwistAxis="0,1,0"
    )) -> (Output_Get:double)
   > Set crouch_legUp_R(crouch_legUp_R:double=GetAlphaFromReferenceTransformDelta(
     SkeletalMeshComponent=Get Body_Mesh()
     BoneName=thigh_r
     BaseName=pelvis
     DeltaComponentType=QuaternionTwist
     InputRange=(X=30,Y=110)
     TwistAxis="0,0,1"
    )) -> (Output_Get:double)
   > Set legOut_R(legOut_R:double=GetAlphaFromReferenceTransformDelta(
     SkeletalMeshComponent=Get Body_Mesh()
     BoneName=thigh_r
     BaseName=pelvis
     DeltaComponentType=QuaternionTwist
     InputRange=(X=0,Y=-40)
     TwistAxis="0,1,0"
    )) -> (Output_Get:double)
   > Set legIn_L(legIn_L:double=GetAlphaFromReferenceTransformDelta(
     SkeletalMeshComponent=Get Body_Mesh()
     BoneName=thigh_l
     BaseName=pelvis
     DeltaComponentType=QuaternionTwist
     InputRange=(X=0,Y=-38)
     TwistAxis="0,1,0"
    )) -> (Output_Get:double)
   n23 = Get Body_Mesh()
   Set scarf_alpha(
    execute=@n1.then_1
    scarf_alpha:double=GetAlphaFromReferenceTransformDelta(SkeletalMeshComponent=@n23, BoneName=head, BaseName=spine_05, DeltaComponentType=QuaternionTwist, InputRange=(X=0,Y=-25), TwistAxis="0,0,1")
   ) -> (Output_Get:double)
   > Set scarf_bck_alpha(scarf_bck_alpha:double=GetAlphaFromReferenceTransformDelta(SkeletalMeshComponent=@n23, BoneName=head, BaseName=spine_05, DeltaComponentType=QuaternionTwist, InputRange=(X=0,Y=45), TwistAxis="0,0,1")) -> (Output_Get:double)
   Event OnBeginCrouching()
   > Set is_crouching(is_crouching:bool=true) -> (Output_Get:bool)
   Event OnEndCrouching()
   > Set is_crouching(is_crouching=false)
   Event BlueprintInitializeAnimation()
   > Set Body_Mesh(Body_Mesh=GetOwnerSkeletalMeshComponent())
   > Set Pawn(Pawn:Fort Pawn Object Reference=TryGetFortPawn()) -> (Output_Get:Fort Pawn Object Reference)
   TryGetPawnOwner()
  ]}
]
variables [
 backbling_active {type=boolean}
 Body_Mesh {type=SkeletalMeshComponent}
 crouch_legUp_R {type=number}
 crouch_moving {type=boolean}
 crouch_pouch {type=number}
 crouch_rigid_alpha {type=number}
 crouch_speed_alpha {type=number}
 crouchAlpha {type=number}
 is_crouching {type=boolean}
 IsDBNO {type=boolean}
 isMovingSide {type=boolean}
 isSkyDivingOn {type=number}
 isWearingCape {type=boolean}
 jump_alpha {type=number}
 legIn_L {type=number}
 legOut_L {type=number}
 legOut_R {type=number}
 legUp_L {type=number}
 legUp_R {type=number}
 move_speed_alpha {type=number}
 moveLeftLegCrossAlpha {type=number}
 moveRighttLegCrossAlpha {type=number}
 not_rigids_active {type=boolean}
 NotDBNO {type=boolean}
 Pawn {type=FortPawn, default=null}
 scarf_alpha {type=number}
 scarf_bck_alpha {type=number}
]
