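# Physics asset for the hair mesh of the female medium Sailor Squad Rose character cosmetic.
# Capsule bodies cover the head and four two-bone dynamic strands: front left and right, side left and right.
# Each strand's first bone is constrained to the head and its second bone to the first, forming short chains with tight swing and twist limits.
# Front tip joints allow wider swing than their roots, while side strands swing asymmetrically.
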PhysicsAsset /BRCosmetics/Characters/Player/Female/Medium/Bodies/F_MED_Sailor_Squad_Rose/Meshes/Parts/F_MED_SSquad_Hair_Physics
bodies {
 dyn_frnt_hair_01_l [{shapeType=Capsule, length=1.75, radius=3.25, center=<1 0 0>, rotation=<rot 90 0 0>}]
 dyn_frnt_hair_01_r [{shapeType=Capsule, length=1.75, radius=2.25, center=<1 0 0>, rotation=<rot 90 0 0>}]
 dyn_frnt_hair_02_l [{shapeType=Capsule, length=3.25, radius=2.5, center=<1 0 0>, rotation=<rot 90 0 0>}]
 dyn_frnt_hair_02_r [{shapeType=Capsule, length=2.75, radius=2.5, center=<2.5 0 0>, rotation=<rot 90 -1.1935 -10.2654>}]
 dyn_hair_side_01_l [{shapeType=Capsule, length=4.505, radius=1.505, rotation=<rot 90 0 0>}]
 dyn_hair_side_01_r [{shapeType=Capsule, length=4.505, radius=1.505, rotation=<rot 90 0 0>}]
 dyn_hair_side_02_l [{shapeType=Capsule, length=4.255, radius=1.755, rotation=<rot 90 0 0>}]
 dyn_hair_side_02_r [{shapeType=Capsule, length=4.505, radius=1.505, rotation=<rot 90 0 0>}]
 head [{shapeType=Capsule, length=9, radius=6.5, rotation=<rot 90 90 96.6375>}]
}
constraints [
 {bone1Name=dyn_frnt_hair_01_r, bone2Name=head, swing1LimitDegrees=10, swing2LimitDegrees=10, twistLimitDegrees=4}
 {bone1Name=dyn_frnt_hair_02_r, bone2Name=dyn_frnt_hair_01_r, swing1LimitDegrees=15, swing2LimitDegrees=15, twistLimitDegrees=4}
 {bone1Name=dyn_frnt_hair_01_l, bone2Name=head, swing1LimitDegrees=10, swing2LimitDegrees=10, twistLimitDegrees=4}
 {bone1Name=dyn_frnt_hair_02_l, bone2Name=dyn_frnt_hair_01_l, swing1LimitDegrees=15, swing2LimitDegrees=15, twistLimitDegrees=4}
 {bone1Name=dyn_hair_side_01_l, bone2Name=head, swing1LimitDegrees=15, swing2LimitDegrees=8, twistLimitDegrees=5}
 {bone1Name=dyn_hair_side_02_l, bone2Name=dyn_hair_side_01_l, swing1LimitDegrees=15, swing2LimitDegrees=8, twistLimitDegrees=5}
 {bone1Name=dyn_hair_side_01_r, bone2Name=head, swing1LimitDegrees=15, swing2LimitDegrees=8, twistLimitDegrees=5}
 {bone1Name=dyn_hair_side_02_r, bone2Name=dyn_hair_side_01_r, swing1LimitDegrees=15, swing2LimitDegrees=8, twistLimitDegrees=5}
]
