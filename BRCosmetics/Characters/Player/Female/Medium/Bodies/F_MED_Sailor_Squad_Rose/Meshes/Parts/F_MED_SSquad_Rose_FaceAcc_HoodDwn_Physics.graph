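# Physics asset for the hood-down face accessory part of the female Sailor Squad Rose cosmetic, covering the back hood and scarf bones.
# Capsule bodies form two chains off a wide spine_05 capsule: hood via a tiny driver body, scarf via a tiny mod body.
# Both intermediate bodies have free swing and twist against the spine.
# Angular limits sit on the outer joints, tighter for the scarf than for the hood's swing.
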PhysicsAsset /BRCosmetics/Characters/Player/Female/Medium/Bodies/F_MED_Sailor_Squad_Rose/Meshes/Parts/F_MED_SSquad_Rose_FaceAcc_HoodDwn_Physics
bodies {
 dyn_bck_hood_01 [{shapeType=Capsule, length=9.75, radius=6.75, center=<6.25 -4 0>, rotation=<rot 90 0 0>}]
 dyn_bck_scarf [{shapeType=Capsule, length=5, radius=5, center=<3 0 0>, rotation=<rot 90 0 0>}]
 dyn_driver_bck_hood_01 [{shapeType=Capsule, length=1, radius=1}]
 dyn_mod_bck_scarf [{shapeType=Capsule, length=1, radius=1}]
 spine_05 [{shapeType=Capsule, length=3.5, radius=25.25, center=<0 15.3297 0>}]
}
constraints [
 {
  bone1Name=dyn_driver_bck_hood_01
  bone2Name=spine_05
  swing1LimitDegrees=45
  swing1Motion=Free
  swing2LimitDegrees=45
  swing2Motion=Free
  twistLimitDegrees=45
  twistMotion=Free
 }
 {bone1Name=dyn_bck_hood_01, bone2Name=dyn_driver_bck_hood_01, swing1LimitDegrees=30, swing2LimitDegrees=30, twistLimitDegrees=5}
 {
  bone1Name=dyn_mod_bck_scarf
  bone2Name=spine_05
  swing1LimitDegrees=45
  swing1Motion=Free
  swing2LimitDegrees=45
  swing2Motion=Free
  twistLimitDegrees=45
  twistMotion=Free
 }
 {bone1Name=dyn_bck_scarf, bone2Name=dyn_mod_bck_scarf, swing1LimitDegrees=5, swing2LimitDegrees=5, twistLimitDegrees=5}
]
