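# Physics asset for the face accessory mesh of the female medium Sailor Squad Rose cosmetic body, covering a hood's top and lip.
# Capsule bodies sit on the head and on two short bone chains, hood top and hood lip, each a parent bone followed by one child.
# Constraints link each chain to the head with tight swing and twist limits, restricting the hood to small movements.
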PhysicsAsset /BRCosmetics/Characters/Player/Female/Medium/Bodies/F_MED_Sailor_Squad_Rose/Meshes/Parts/F_MED_SSquad_Rose_FaceAcc_Physics
bodies {
 dyn_hood_lip_01 [{shapeType=Capsule, length=5.25, radius=3, center=<0.2 0 0>, rotation=<rot 90 0 0>}]
 dyn_hood_top_01 [{shapeType=Capsule, length=5, radius=1.75, center=<0.1 0 0>, rotation=<rot 90 0 0>}]
 dyn_parent_hood_lip [{shapeType=Capsule, length=5.75, radius=5, center=<0.1 0 0>, rotation=<rot 90 0 0>}]
 dyn_parent_hood_top [{shapeType=Capsule, length=5.75, radius=5, center=<0.15 0 0>, rotation=<rot 90 0 0>}]
 head [{shapeType=Capsule, length=6.25, radius=6, center=<1.0526 2.7759 0.0001>, rotation=<rot 90 20.556 13.5014>}]
}
constraints [
 {bone1Name=dyn_parent_hood_top, bone2Name=head, swing1LimitDegrees=10, swing2LimitDegrees=10, twistLimitDegrees=3}
 {bone1Name=dyn_hood_top_01, bone2Name=dyn_parent_hood_top, swing1LimitDegrees=10, swing2LimitDegrees=10, twistLimitDegrees=3}
 {bone1Name=dyn_parent_hood_lip, bone2Name=head, swing1LimitDegrees=3, swing2LimitDegrees=3, twistLimitDegrees=3}
 {bone1Name=dyn_hood_lip_01, bone2Name=dyn_parent_hood_lip, swing1LimitDegrees=8, swing2LimitDegrees=15, twistLimitDegrees=3}
]
