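# Post-processes the Sailor Squad Rose face accessory part, animating hood, scarf and hair over the pose copied from the attached parent mesh.
# Head look and tilt alphas, measured as twist against spine_05, drive smoothed additive offsets on scarf and front hood bones.
# Three RigidBody passes add physics, followed by hood shaping offsets.
# A worn backbling switches off the hood-down simulation and blends in shrinking of the back hood bones.
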AnimBlueprint /BRCosmetics/Characters/Player/Female/Medium/Bodies/F_MED_Sailor_Squad_Rose/Meshes/Parts/F_MED_SSquad_Rose_FaceAcc_AnimBP
target_skeleton=&/F_MED_SSquad_Rose_FaceAcc_Skeleton
graphs [
 AnimGraph {nodes [
   n1 = "Constraint - dyn_parent_hooddwn_frt_l (PP)"!(
    ComponentPose=Local To Component(LocalPose=Copy Pose From Mesh() {bUseAttachedParent=true})
    Alpha=1
    ConstraintWeights_0=0.4081
    ConstraintWeights_1:float=0.5919
   ) -> (Pose) {
    alphaBoolBlend {blendOption=Linear}
    boneToModify {boneName=dyn_parent_hooddwn_frt_l}
    constraintSetup [{targetBone {boneName=spine_05}}, {targetBone {boneName=clavicle_l}}]
    constraintWeights [1,1]
   }
   n2 = "Transform (Modify) Bone - Bone: dyn_mod_bck_scarf"(ComponentPose=@n1, Alpha=1, Translation="0,0,0", Rotation="0,6,0", Scale="1,1,1") {
    rotationMode=BMM_Additive
    rotationSpace=BCS_BoneSpace
    scaleMode=BMM_Additive
    scaleSpace=BCS_BoneSpace
    translationMode=BMM_Additive
    translationSpace=BCS_BoneSpace
    alphaBoolBlend {blendOption=Linear}
    alphaScaleBiasClamp {bInterpResult=true, interpSpeedDecreasing=15, interpSpeedIncreasing=150}
    boneToModify {boneName=dyn_mod_bck_scarf}
    rotation {yaw=6}
   }
   n3 = "Transform (Modify) Bone - Bone: dyn_mod_bck_scarf"(
    ComponentPose=@n2
    Alpha=Get head delta_look_up() -> (head delta_look_up:double)
    Translation="0,0,0"
    Rotation="0,13,0"
    Scale="0.95,0.95,0.95"
   ) {
    rotationMode=BMM_Additive
    rotationSpace=BCS_BoneSpace
    scaleMode=BMM_Additive
    scaleSpace=BCS_BoneSpace
    translationMode=BMM_Additive
    translationSpace=BCS_BoneSpace
    alphaBoolBlend {blendOption=Linear}
    alphaScaleBiasClamp {bInterpResult=true, interpSpeedDecreasing=15, interpSpeedIncreasing=150}
    boneToModify {boneName=dyn_mod_bck_scarf}
    rotation {yaw=13}
    scale=<0.95 0.95 0.95>
   }
   n4 = Get head delta_look_lft() -> (head delta_look_lft:double)
   n5 = "Transform (Modify) Bone - Bone: dyn_hood_frt_01_l"(ComponentPose=@n3, Alpha=@n4, Translation="-2.5,0,0", Rotation="-16,8,0.000", Scale="1,1,1") {
    rotationMode=BMM_Additive
    rotationSpace=BCS_BoneSpace
    translationMode=BMM_Additive
    translationSpace=BCS_BoneSpace
    alphaBoolBlend {blendOption=Linear}
    alphaScaleBiasClamp {bInterpResult=true, interpSpeedDecreasing=25, interpSpeedIncreasing=25}
    boneToModify {boneName=dyn_hood_frt_01_l}
    rotation {pitch=-16, yaw=8}
    translation {x=-2.5}
   }
   n6 = "Transform (Modify) Bone - Bone: dyn_hood_frt_01_r"(ComponentPose=@n5, Alpha=@n4, Translation="0,0,0", Rotation="-15,5,5", Scale="1,1,1") {
    rotationMode=BMM_Additive
    rotationSpace=BCS_BoneSpace
    translationMode=BMM_Additive
    translationSpace=BCS_BoneSpace
    alphaBoolBlend {blendOption=Linear}
    alphaScaleBiasClamp {bInterpResult=true, interpSpeedDecreasing=25, interpSpeedIncreasing=25}
    boneToModify {boneName=dyn_hood_frt_01_r}
    rotation=<rot -15 5 5>
   }
   n7 = Get head delta_look_rgt() -> (head delta_look_rgt:double)
   n8 = "Transform (Modify) Bone - Bone: dyn_hood_frt_01_r"(ComponentPose=@n6, Alpha=@n7, Translation="0,0,0", Rotation="32,32,0", Scale="1,1,1") {
    rotationMode=BMM_Additive
    rotationSpace=BCS_BoneSpace
    translationMode=BMM_Additive
    translationSpace=BCS_BoneSpace
    alphaBoolBlend {blendOption=Linear}
    alphaScaleBiasClamp {bInterpResult=true, interpSpeedDecreasing=25, interpSpeedIncreasing=25}
    boneToModify {boneName=dyn_hood_frt_01_r}
    rotation {pitch=32, yaw=32}
   }
   n9 = "Transform (Modify) Bone - Bone: dyn_hood_frt_01_l"(ComponentPose=@n8, Alpha=@n7, Translation="0,0,0", Rotation="10,10,-5", Scale="1,1,1") {
    rotationMode=BMM_Additive
    rotationSpace=BCS_BoneSpace
    translationMode=BMM_Additive
    translationSpace=BCS_BoneSpace
    alphaBoolBlend {blendOption=Linear}
    alphaScaleBiasClamp {bInterpResult=true, interpSpeedDecreasing=25, interpSpeedIncreasing=25}
    boneToModify {boneName=dyn_hood_frt_01_l}
    rotation=<rot 10 10 -5>
   }
   n10 = "Transform (Modify) Bone - Bone: dyn_hood_frt_01_r"(
    ComponentPose=@n9
    Alpha=Get head delta_tilt_right() -> (head delta_tilt_right:double)
    Translation="0,0,0"
    Rotation="15,15,0"
    Scale="1,1,1"
   ) {
    rotationMode=BMM_Additive
    rotationSpace=BCS_BoneSpace
    translationMode=BMM_Additive
    translationSpace=BCS_BoneSpace
    alphaBoolBlend {blendOption=Linear}
    alphaScaleBiasClamp {bInterpResult=true, interpSpeedDecreasing=25, interpSpeedIncreasing=25}
    boneToModify {boneName=dyn_hood_frt_01_r}
    rotation {pitch=15, yaw=15}
   }
   n11 = "Transform (Modify) Bone - Bone: dyn_hood_frt_01_r"(ComponentPose=@n10, Alpha=Get head delta_look_up(), Translation="0,0,0", Rotation="15,15,0", Scale="1,1,1") {
    rotationMode=BMM_Additive
    rotationSpace=BCS_BoneSpace
    translationMode=BMM_Additive
    translationSpace=BCS_BoneSpace
    alphaBoolBlend {blendOption=Linear}
    alphaScaleBiasClamp {bInterpResult=true, interpSpeedDecreasing=25, interpSpeedIncreasing=25}
    boneToModify {boneName=dyn_hood_frt_01_r}
    rotation {pitch=15, yaw=15}
   }
   n12 = Get head delta_look_dwn() -> (head delta_look_dwn:double)
   n13 = "Transform (Modify) Bone - Bone: dyn_hood_frt_01_r"(ComponentPose=@n11, Alpha=@n12, Translation="0,0,0", Rotation="15,15,0", Scale="1,1,1") {
    rotationMode=BMM_Additive
    rotationSpace=BCS_BoneSpace
    translationMode=BMM_Additive
    translationSpace=BCS_BoneSpace
    alphaBoolBlend {blendOption=Linear}
    alphaScaleBiasClamp {bInterpResult=true, interpSpeedDecreasing=25, interpSpeedIncreasing=25}
    boneToModify {boneName=dyn_hood_frt_01_r}
    rotation {pitch=15, yaw=15}
   }
   n14 = "Transform (Modify) Bone - Bone: dyn_hood_frt_01_l"(ComponentPose=@n13, Alpha=@n12, Translation="0,0,0", Rotation="10,10,-5", Scale="1,1,1") {
    rotationMode=BMM_Additive
    rotationSpace=BCS_BoneSpace
    translationMode=BMM_Additive
    translationSpace=BCS_BoneSpace
    alphaBoolBlend {blendOption=Linear}
    alphaScaleBiasClamp {bInterpResult=true, interpSpeedDecreasing=25, interpSpeedIncreasing=25}
    boneToModify {boneName=dyn_hood_frt_01_l}
    rotation=<rot 10 10 -5>
   }
   n15 = RigidBody(ComponentPose=@n14) {
    bClampLinearTranslationLimitToRefPose=true
    bOverrideWorldGravity=true
    overridePhysicsAsset=&/F_MED_SSquad_Rose_FaceAcc_Physics
    simulationSpace=BaseBoneSpace
    alphaBoolBlend {blendOption=Linear}
    baseBoneRef {boneName=neck_02}
    componentAppliedLinearAccClamp=<1100 1100 1000>
    componentLinearAccScale=<0.5 0.5 0.5>
    componentLinearVelScale=<0.5 0.5 0.5>
    overrideWorldGravity {z=-980}
    simSpaceSettings {
     maxAngularAcceleration=500
     maxAngularVelocity=3
     maxLinearAcceleration=400
     maxLinearVelocity=250
     velocityScaleZ=0
     worldAlpha=1
     externalLinearDragV=<0.1 0.1>
    }
   }
   n16 = RigidBody(ComponentPose=@n15) {
    bClampLinearTranslationLimitToRefPose=true
    bOverrideWorldGravity=true
    overridePhysicsAsset=&/F_MED_SSquad_Hair_Physics
    simulationSpace=BaseBoneSpace
    alphaBoolBlend {blendOption=Linear}
    baseBoneRef {boneName=root}
    componentAppliedLinearAccClamp=<1100 1100 1100>
    componentLinearAccScale=<0.5 0.5 0.6>
    componentLinearVelScale=<0.5 0.5 0.6>
    overrideWorldGravity {z=-950}
   }
   n17 = RigidBody(ComponentPose=@n16, Alpha=Get backbling_alpha() -> (backbling_alpha:double)) {
    bOverrideWorldGravity=true
    overridePhysicsAsset=&/F_MED_SSquad_Rose_FaceAcc_HoodDwn_Physics
    simulationSpace=BaseBoneSpace
    alphaBoolBlend {blendOption=Linear}
    baseBoneRef {boneName=root}
    componentAppliedLinearAccClamp=<3000 3000 3000>
    componentLinearAccScale=<0.35 0.35 0.425>
    componentLinearVelScale=<0.35 0.35 0.425>
    overrideWorldGravity {z=-980}
   }
   n18 = "Transform (Modify) Bone - Bone: dyn_hood_lip_01"(ComponentPose=@n17, Alpha=1, Translation="0,0,1", Rotation="0,0,0", Scale="1,1,1.1") {
    scaleMode=BMM_Additive
    translationMode=BMM_Additive
    alphaBoolBlend {blendOption=Linear}
    boneToModify {boneName=dyn_hood_lip_01}
    scale {z=1.1}
    translation {z=1}
   }
   n19 = "Transform (Modify) Bone - Bone: dyn_hood_top_01"(ComponentPose=@n18, Alpha=1, Translation="0,-3,1", Rotation="0,0,0", Scale="1,1.2,1.2") {
    scaleMode=BMM_Additive
    scaleSpace=BCS_WorldSpace
    translationMode=BMM_Additive
    alphaBoolBlend {blendOption=Linear}
    boneToModify {boneName=dyn_hood_top_01}
    scale {y=1.2, z=1.2}
    translation {y=-3, z=1}
   }
   n20 = Get backbling_hood_alpha() -> (backbling_hood_alpha:double)
   n21 = "Transform (Modify) Bone - Bone: dyn_bck_hood_02"(ComponentPose=@n19, Alpha=@n20, Translation="-0.5,1,0", Rotation="0,0,0", Scale="0.7,0.7,0.7") {
    scaleMode=BMM_Additive
    scaleSpace=BCS_BoneSpace
    translationMode=BMM_Additive
    translationSpace=BCS_BoneSpace
    alphaBoolBlend {blendOption=Linear}
    boneToModify {boneName=dyn_bck_hood_02}
    scale=<0.7 0.7 0.7>
    translation=<-0.5 1>
   }
   n22 = "Transform (Modify) Bone - Bone: dyn_parent_bck_hood_01"(ComponentPose=@n21, Alpha=@n20, Translation="0,2,0", Rotation="0,0,0", Scale="0.9,0.9,0.9") {
    scaleMode=BMM_Additive
    scaleSpace=BCS_BoneSpace
    translationMode=BMM_Additive
    translationSpace=BCS_BoneSpace
    alphaBoolBlend {blendOption=Linear}
    boneToModify {boneName=dyn_parent_bck_hood_01}
    scale=<0.9 0.9 0.9>
    translation {y=2}
   }
   Output Pose(Result=Component To Local(ComponentPose=@n22)) {layerGroup=None}
  ]}
 EventGraph {nodes [
   Event BlueprintUpdateAnimation()
   > Get!(execute:Exec) -> (then:Exec, else:Exec, Skel_Mesh:Skeletal Mesh Component Object Reference)
   > n1 = Sequence()
   n2 = CheckForBackbling()
   > Set isWearingBackbling(isWearingBackbling=OR Boolean(A=@n2, B=@n2.bIsBackblingLong))
   > n3 = Branch(Condition=Get isWearingBackbling())
   > Set backbling_alpha(backbling_alpha:double=0) -> (Output_Get:double)
   > Set backbling_hood_alpha(backbling_hood_alpha:double=1) -> (Output_Get:double)
   Set backbling_alpha(execute=@n3.else, backbling_alpha=1)
   > Set backbling_hood_alpha(backbling_hood_alpha=0)
   Set head delta_look_up(
    execute=@n1.then_1
    head delta_look_up:double=GetAlphaFromReferenceTransformDelta(
     SkeletalMeshComponent=Get Skel_Mesh()
     BoneName=head
     BaseName=spine_05
     DeltaComponentType=QuaternionTwist
     InputRange=(X=0,Y=46)
     TwistAxis="0,0,1"
    )
   ) -> (Output_Get:double)
   > Set head delta_look_dwn(head delta_look_dwn:double=GetAlphaFromReferenceTransformDelta(
     SkeletalMeshComponent=Get Skel_Mesh()
     BoneName=head
     BaseName=spine_05
     DeltaComponentType=QuaternionTwist
     InputRange=(X=0,Y=-20)
     TwistAxis="0,0,1"
    )) -> (Output_Get:double)
   > Set head delta_look_lft(head delta_look_lft:double=GetAlphaFromReferenceTransformDelta(
     SkeletalMeshComponent=Get Skel_Mesh()
     BoneName=head
     BaseName=spine_05
     DeltaComponentType=QuaternionTwist
     InputRange=(X=0,Y=50)
     TwistAxis="1,0,0"
    )) -> (Output_Get:double)
   > Set head delta_look_rgt(head delta_look_rgt:double=GetAlphaFromReferenceTransformDelta(
     SkeletalMeshComponent=Get Skel_Mesh()
     BoneName=head
     BaseName=spine_05
     DeltaComponentType=QuaternionTwist
     InputRange=(X=0,Y=-85)
     TwistAxis="1,0,0"
    )) -> (Output_Get:double)
   > Set head delta_tilt_right(head delta_tilt_right:double=GetAlphaFromReferenceTransformDelta(
     SkeletalMeshComponent=Get Skel_Mesh()
     BoneName=head
     BaseName=spine_05
     DeltaComponentType=QuaternionTwist
     InputRange=(X=0,Y=38)
     TwistAxis="0,1,0"
    )) -> (Output_Get:double)
   Event BlueprintInitializeAnimation()
   > Set Skel_Mesh(Skel_Mesh=GetPartSkeletalMeshComponent())
   Event OnCharacterCustomizationLoaded()
  ]}
]
variables [
 backbling_alpha {type=number, default=1}
 backbling_hood_alpha {type=number}
 head delta_look_dwn {type=number}
 head delta_look_lft {type=number}
 head delta_look_rgt {type=number}
 head delta_look_up {type=number}
 head delta_tilt_right {type=number}
 isWearingBackbling {type=boolean}
 Skel_Mesh {type=SkeletalMeshComponent}
]
